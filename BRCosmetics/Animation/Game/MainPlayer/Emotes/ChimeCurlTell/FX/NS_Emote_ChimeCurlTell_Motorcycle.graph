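# Niagara effect for the motorcycle in the ChimeCurlTell traversal emote, adding lights and dust.
# Light holds three mesh-attached, teal-green particle lights, pulsed by a waveform and scaled by time of day and User.LightVisibility.
# Dust sprites spawn at mesh offsets, their rate driven by an anim-curve parameter gated on owner speed, which also blends drag.
# Front-end modifiers override spawn rate, velocity, dust opacity and light exponent for menu display.
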NiagaraSystem /BRCosmetics/Animation/Game/MainPlayer/Emotes/ChimeCurlTell/FX/NS_Emote_ChimeCurlTell_Motorcycle
dependencies {
 usedDataInterfaces [&NiagaraDataInterfaceCurve, &NiagaraDataInterfaceSkeletalMesh]
 usedDynamicInputs [
  &UniformRangedLinearColor
  &Vector2DFromFloat
  &MakeLinearColorFromVectorAndFloat
  &MakeVectorFromLinearColorRGB
  &FNTimeOfDayColor
  &RandomRangeLinearColor
 ]
 usedModules [
  &SystemState
  &self:SystemSpawnScript.SystemScriptSource.SystemScriptGraph.NiagaraNodeAssignment_2.SetVariables_#1
  &self:SystemSpawnScript.SystemScriptSource.SystemScriptGraph.NiagaraNodeAssignment_0.SetVariables_#2
 ]
}
emitters {
 Light {
  emitterUpdateScript [
   &EmitterState(
    Inactive Response=::NewEnumerator1
    Life Cycle Mode=::NewEnumerator1
    Loop Behavior=::NewEnumerator1
    Loop Duration Mode=::NewEnumerator1
    UseLoopDelay=false
   )
   &SpawnBurst_Instantaneous(Spawn Count=3, Use Spawn Probability=false, UseLoopCountLimit=false)
   /Niagara/Noise/Waveform(
    EnableExponent1=false
    Global Amplitude Scale=0
    GlobalClampMax=false
    GlobalClampMin=false
    UseInput1=false
    UsePhase1=false
    "Waveform [1]":ENiagara_Waveforms=::NewEnumerator12
    "[1] Frequency"=7
   )
   /SetVariables(Emitter.Flicker:NiagaraFloat=&Add_Float(A=&Multiply_Float(A=$Emitter.Waveform.Waveform, B=0.5), B=0.5)) {isSetParameters=true}
  ]
  particleSpawnScript [&PartitionParticles(
    Number Of Partitions=3
    Particles Per Partition=1
    Partition Mode:ENiagara_PartitionMode_v2=::NewEnumerator1
    Write Partition to Ribbon ID=::NewEnumerator0
   )]
  particleUpdateScript [
   &ParticleState(
    Kill Particles When Lifetime Has Elapsed=false
    Let Infinitely Lived Particles Die When Emitter Deactivates=true
    Loop Particles Lifetime=true
   )
   &SkeletalMeshLocation(
    Bone / Socket Index=2
    Bone Sampling Mode=::NewEnumerator4
    MaskSpawnGroup=false
    Orientation Sampling=::NewEnumerator1
    Position Sampling=::NewEnumerator1
    Sampled Position Offset=&MergeHelper_VectorSwitch(
     0=&MakeVector(X=0, Y=25, Z=15)
     1=<0 -40 -15>
     2=<0 45 -15>
     3=?
     Index=$Particles.PartitionParticles.Partition
     Num=::NewEnumerator1
    )
    Skeletal Mesh=$System.Skeletal Mesh
   )
   &FN_Light_Attributes(
    Color=<rgba 0.231 5 2.1335 1>
    Create Color Attribute=true
    Create Exponent Attribute=true
    Create Position Attribute=true
    Create Radius Attribute=true
    Enable Particle Light Enable=true
    Enable Visibility Tag=true
    Exponent:NiagaraFloat=&FN_FrontEndFloatModifier(
     Gameplay=&MergeHelper_FloatSwitch(0=2.5, 1=1.5, 2=1.5, 3=?, Index=$Particles.PartitionParticles.Partition, Num=::NewEnumerator1)
     Front End=&MergeHelper_FloatSwitch(0=13, 1=4, 2=4, 3=?, Index=$Particles.PartitionParticles.Partition, Num=::NewEnumerator1)
    )
    Light Enabled=true
    Position=&MergeHelper_VectorSwitch(0=<0 0 50>, 1=<0 -46 15>, 2=<0 46 15>, 3=?, Index=$Particles.PartitionParticles.Partition, Num=::NewEnumerator1)
    Radius=&MergeHelper_FloatSwitch(0=50, 1=35, 2=35, 3=?, Index=$Particles.PartitionParticles.Partition, Num=::NewEnumerator1)
    Scale Color=&Multiply_Float(
     A=$User.LightVisibility
     B=&Multiply_Float(
      A=/Niagara/DynamicInputs/Helpers/Waveform("Waveform [1]"=::NewEnumerator0, "[1] Frequency"=3, "[1] Amplitude Min/Max"=<0.7 1>, "[1] Scale"=2.6368)
      B=&FNTimeOfDayScalar(Night Scale=0.45, Morning Scale=0.6, Day Scale=0.7, Evening Scale=0.6)
     )
    )
    VisibilityTag=1
   )
  ]
  rendererClasses [&NiagaraLightRendererProperties]
  renderers [{rendererClass=&NiagaraLightRendererProperties, rendererIndex=0}]
 }
 dust_traversal002 {
  emitterSpawnScript [/SetVariables(Emitter.FE Particle Spawn Rate=60, Emitter.Max Traversal Speed=300, Emitter.Particle Spawn Rate=60) {isSetParameters=true}]
  emitterUpdateScript [
   &EmitterState()
   &SpawnRate(
    SpawnRate=&FN_FrontEndFloatModifier(
     Gameplay=&MakeCustomFloatFromBool(
      Bool=&SetBoolByFloatComparison(A=&VectorLength(Vector=$User.OwnerLocalVelocity), B=150)
      True Float=$User.Anim_curve_spawnRate
      False Float=0
     )
     Front End=35
    )
    Use Spawn Probability=false
   ) {module=SpawnRate002}
  ]
  particleSpawnScript [
   &PartitionParticles()
   &InitializeParticle(
    Color=$System.Global color
    Lifetime=&MergeHelper_FloatSwitch(
     0=&RandomRangeFloat(Minimum=0.3, Maximum=0.5)
     1=&RandomRangeFloat(Minimum=0.45, Maximum=0.7)
     2=?
     3=?
     Index=$Particles.PartitionParticles.Partition
     Num=::NewEnumerator0
    )
    Sprite Rotation Mode=::NewEnumerator2
    Sprite Size Mode=::NewEnumerator1
    Uniform Sprite Size Max=65
    Uniform Sprite Size Min=45
    UsePositionOffset=true
    Write Material Random=true
   ) {module=InitializeParticle001}
   &SkeletalMeshLocation(
    Bone Sampling Mode=::NewEnumerator4
    MaskSpawnGroup=false
    Sampled Position Offset=&MaskVector(Mask Source=0, Mask Parameter=?, Mask=$Particles.PartitionParticles.Partition, Masked Vector=<0 25 -15>, Passthrough Vector=<0 -60 -20>)
    Sampled Position Offset Coordinate Space:ENiagaraSkelMeshTransforms=::NewEnumerator2
    Skeletal Mesh=$System.Skeletal Mesh
   )
   &CameraOffset(Camera Offset Amount=2)
   &AddVelocity(
    Velocity=&MergeHelper_VectorSwitch(
     0=&FN_FrontEndVector3Modifier(Gameplay=<0 -3 0.3>, Front End=<0 -10 0>)
     1=<0 -3 0.5>
     2=?
     3=?
     Index=$Particles.PartitionParticles.Partition
     Num=::NewEnumerator0
    )
    Velocity Speed Scale=&RandomRangeFloat(Minimum=300, Maximum=500)
   ) {module=AddVelocity002}
   &ShapeLocation(
    MaskSpawnGroup=false
    Override Seed=false
    Sphere Radius=&FloatFromCurve(FloatCurve=DataInterface() {propertyValues {Curve {keys [<at 0 0>,<at 1 5>]}}})
    UseOwnerScale=true
   )
  ]
  particleUpdateScript [
   &ParticleState()
   &ScaleSpriteSize(
    Scale Sprite Size Mode=::NewEnumerator0
    Uniform Scale Factor=&MergeHelper_FloatSwitch(
     0=&FloatFromCurve(
      FloatCurve=DataInterface() {propertyValues {Curve {keys [
          {arriveTangent=0.8753, interpMode=RCIM_Cubic, leaveTangent=0.8753, tangentMode=RCTM_User, time=0, value=0.5}
          {arriveTangent=0.0238, interpMode=RCIM_Cubic, leaveTangent=0.0238, tangentMode=RCTM_User, time=1, value=1}
         ]}}}
      Scale Curve=1.3
     )
     1=&FloatFromCurve(
      FloatCurve=DataInterface() {propertyValues {Curve {keys [
          {arriveTangent=0.8753, interpMode=RCIM_Cubic, leaveTangent=0.8753, tangentMode=RCTM_User, time=0, value=0.5}
          {arriveTangent=0.0238, interpMode=RCIM_Cubic, leaveTangent=0.0238, tangentMode=RCTM_User, time=0.9843, value=1.7576}
         ]}}}
      Scale Curve=1.5
     )
     2=?
     3=?
     Index=$Particles.PartitionParticles.Partition
     Num=::NewEnumerator0
    )
   )
   &ScaleColor(
    Scale Alpha=&FloatFromCurve(
     FloatCurve=DataInterface() {propertyValues {Curve {keys [
         {arriveTangent=0.0878, interpMode=RCIM_Cubic, leaveTangent=0.0878, tangentMode=RCTM_User, time=0, value=0}
         {
          arriveTangent=0.0878
          arriveTangentWeight=0.0754
          interpMode=RCIM_Cubic
          leaveTangent=0.0878
          leaveTangentWeight=0.2297
          tangentMode=RCTM_User
          time=0.1
          value=1
         }
         {arriveTangent=0.2106, interpMode=RCIM_Cubic, leaveTangent=-3.5553, tangentMode=RCTM_Break, time=0.4045, value=1}
         {arriveTangent=-0.0368, interpMode=RCIM_Cubic, leaveTangent=-2.9749, tangentMode=RCTM_Break, time=0.9, value=0}
        ]}}}
     Scale Curve=&FN_FrontEndFloatModifier(
      Gameplay=&FNTimeOfDayScalar(Night Scale=0.05, Morning Scale=0.08, Day Scale=0.07, Evening Scale=0.08)
      Front End=0.15
     )
    )
    Scale RGB=&VectorFromFloat(Value=&RandomRangeFloat(Minimum=0.7, Maximum=1))
    ScaleA=true
    ScaleRGB=true
   )
   &Drag(
    Drag=&FloatFromCurve(
     FloatCurve=DataInterface() {propertyValues {Curve {
        defaultValue=1
        keys [{interpMode=RCIM_Cubic, tangentMode=RCTM_User, time=0, value=1}, {interpMode=RCIM_Cubic, tangentMode=RCTM_User, time=0.65, value=0.75}]
       }}}
     Scale Curve=&Lerp_Float(
      A=&MergeHelper_FloatSwitch(0=50, 1=20, 2=?, 3=?, Index=$Particles.PartitionParticles.Partition, Num=::NewEnumerator0)
      B=20
      Alpha=&RemapRange(
       Input Value To Remap=&VectorLength(Vector=$User.OwnerLocalVelocity)
       Input Min=0
       Input Max=$Emitter.Max Traversal Speed
       Output Min=0
       Output Max=1
      )
     )
    )
    Rotational Drag=5
    Use Linear Drag=true
    Use Rotational Drag=true
   ) {module=Drag001}
   &SolveForcesAndVelocity(Clamp Velocity=false, Limit Acceleration=false)
   &DynamicMaterialParameters(
    Index 0 Param 1=&Add_Float(
     A=&RandomRangeFloat(Minimum=0, Maximum=1)
     B=&FloatFromCurve(
      FloatCurve=DataInterface() {propertyValues {Curve {keys [
          <at 0 0 RCIM_Cubic>
          {arriveTangent=1.5947, interpMode=RCIM_Cubic, leaveTangent=1.5947, tangentMode=RCTM_User, time=0.2557, value=0.5257}
          <at 1 1 RCIM_Cubic>
         ]}}}
      CurveIndex=$Particles.Age
      Scale Curve=0.25
     )
    )
    Module.Index0Param0Write=true
    Module.Index0Param1Write=true
    Module.Index0Param2Write=true
    Module.Index0Param3Write=true
   )
  ]
  rendererClasses [&NiagaraSpriteRendererProperties]
  renderers [
   {rendererClass=&NiagaraSpriteRendererProperties}
   {rendererClass=&NiagaraSpriteRendererProperties, rendererIndex=1}
  ]
 }
}
summary {
 systemName=NS_Emote_ChimeCurlTell_Motorcycle
 emitters [
  {emitterName=Light, rendererClasses [&NiagaraLightRendererProperties]}
  {emitterName=dust_traversal002, rendererClasses [&NiagaraSpriteRendererProperties]}
 ]
 userVariables [
  User.Anim_curve_LightPos {type=NiagaraFloat, defaultValue=0, description=""}
  User.Anim_curve_spawnRate {type=NiagaraFloat, defaultValue=0, description=""}
  User.bIsFrontEnd {type=NiagaraBool}
  User.bIsFrontEndPreview {type=NiagaraBool}
  User.LightVisibility {type=NiagaraFloat, defaultValue=1, description=""}
  User.OwnerLocalVelocity {type=Vector3f, description="", defaultValue {struct=Vector3f, value=<0 0 0>}}
 ]
}
