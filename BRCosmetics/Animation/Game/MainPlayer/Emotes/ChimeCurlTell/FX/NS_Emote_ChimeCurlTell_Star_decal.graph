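# Niagara system for the ChimeCurlTell emote that creates a green-teal star impact on the ground with flare, glass shatter and sparks.
# A merged flare emitter layers glows with a light, while spawn groups drive twinkling sprite and mesh shards alongside GPU curl-noise sparks.
# The star samples the owner's skeleton, fits to ground slope via trace and uses BlendSpace between world and local positioning.
# Time-of-day and front-end modifiers adjust intensity and velocity.
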NiagaraSystem /BRCosmetics/Animation/Game/MainPlayer/Emotes/ChimeCurlTell/FX/NS_Emote_ChimeCurlTell_Star_decal
dependencies {
 usedDataInterfaces [
  &NiagaraDataInterfaceCurve
  &NiagaraDataInterfaceVector2DCurve
  &NiagaraDataInterfaceColorCurve
  &NiagaraDataInterfaceArrayColor
  &NiagaraDataInterfaceSpriteRendererInfo
  &NiagaraDataInterfaceCollisionQuery
  &NiagaraDataInterfaceDebugDraw
 ]
 usedDynamicInputs [&FNTimeOfDayColor]
 usedModules [
  &SystemState
  &self:SystemSpawnScript.SystemScriptSource.SystemScriptGraph.NiagaraNodeAssignment_1.SetVariables_#1
  &self:SystemSpawnScript.SystemScriptSource.SystemScriptGraph.NiagaraNodeAssignment_2.SetVariables_#2
  &BlendSpaceHelper
 ]
}
emitters {
 Flare_impact_merged {
  emitterUpdateScript [
   &EmitterState(Life Cycle Mode=::NewEnumerator1, Loop Behavior=::NewEnumerator1, Loop Duration=1.5, UseLoopDelay=false)
   &SpawnBurst_Instantaneous(Spawn Group=5, Use Spawn Probability=false, UseLoopCountLimit=false) {module=SpawnBurst_Instantaneous005}
   &SpawnBurst_Instantaneous(Spawn Group=4, Use Spawn Probability=false, UseLoopCountLimit=false) {module=SpawnBurst_Instantaneous004}
   &SpawnBurst_Instantaneous(Spawn Group=3, Use Spawn Probability=false, UseLoopCountLimit=false) {module=SpawnBurst_Instantaneous003}
   &SpawnBurst_Instantaneous(Spawn Group=2, Use Spawn Probability=false, UseLoopCountLimit=false) {module=SpawnBurst_Instantaneous002}
   &SpawnBurst_Instantaneous(Spawn Count=1, Spawn Group=1, Use Spawn Probability=false) {module=SpawnBurst_Instantaneous001}
   &SpawnBurst_Instantaneous(Use Spawn Probability=false, UseLoopCountLimit=false)
   /SetVariables(Emitter.DepthFade:NiagaraFloat=&FN_FrontEndFloatModifier(Gameplay=500, Front End=300)) {isSetParameters=true}
  ]
  particleSpawnScript [
   /SetVariables(Particles.VisibilityTag=&SpawnGroup()) {isSetParameters=true}
   &FN_SetSpawnGroup(Number of Spawn Groups=5)
   &InitializeParticle(
    Color Mode=::NewEnumerator0
    Lifetime=&MergeHelper_FloatSwitch(0=0.2, 1=0.3, 2=0.3, 3=0.2, 4=0.2, Index=&SpawnGroup(), Num=::NewEnumerator3)
    Position Offset=<-1.1765 -3.6875 3>
    Position Offset Coordinate Space=::Local
    UsePositionOffset=true
    Write Material Random=true
   )
  ]
  particleUpdateScript [
   &ParticleState()
   /SetVariables(
    "Particles.Hot flare Color(0)":LinearColor=&MakeLinearColorFromVectorAndFloat(
     "Vector (RGB)"=&MakeVectorFromLinearColorRGB(Color=&ColorFromCurve(ColorCurve=DataInterface() {propertyValues {
         AlphaCurve {keys [<at 0 1>,<at 0.2388 1>,<at 1 0>]}
         BlueCurve {keys [<at 0 0>,<at 1 1>]}
         GreenCurve {keys [<at 0 1>,<at 1 0.1667>]}
         RedCurve {keys [<at 0 0>,<at 1 0>]}
        }}))
     "Float (Alpha)"=&FloatFromCurve(FloatCurve=DataInterface() {propertyValues {Curve {keys [{arriveTangent=-3.3102, interpMode=RCIM_Cubic, time=0, value=0}, <at 0.5 1.5 RCIM_Cubic>, <at 1 0 RCIM_Cubic>]}}})
    )
    "Particles.Soft Glow Color  (1)":LinearColor=&MakeLinearColorFromVectorAndFloat(
     "Vector (RGB)"=&MakeVectorFromLinearColorRGB(Color=&ColorFromCurve(ColorCurve=DataInterface() {propertyValues {
         AlphaCurve {keys [<at 0 1>,<at 0.2388 1>,<at 1 0>]}
         BlueCurve {keys [<at 0 0>,<at 1 1>]}
         GreenCurve {keys [<at 0 1>,<at 1 0.1667>]}
         RedCurve {keys [<at 0 0>,<at 1 0>]}
        }}))
     "Float (Alpha)"=&FloatFromCurve(
      FloatCurve=DataInterface() {propertyValues {Curve {keys [{interpMode=RCIM_Cubic, tangentMode=RCTM_User, time=0, value=1}, {interpMode=RCIM_Cubic, tangentMode=RCTM_User, time=1, value=0}]}}}
      Scale Curve=0.5
     )
    )
    "Particles.Soft Glow Size (1)":Vector2f=&Vector2DFromFloat(Value=&Multiply_Float(
      A=50
      B=&FloatFromCurve(
       FloatCurve=DataInterface() {propertyValues {Curve {keys [
           {arriveTangent=2.6083, interpMode=RCIM_Cubic, leaveTangent=0.4318, tangentMode=RCTM_User, time=0, value=0.5}
           {arriveTangent=1.3014, interpMode=RCIM_Cubic, leaveTangent=1.3014, tangentMode=RCTM_User, time=0.3, value=0.8}
           {arriveTangent=0.1331, leaveTangent=0.1331, time=1, value=1.2}
          ]}}}
       Scale Curve=1.25
      )
     ))
    "Particles.Sparkly Sphere Color (2)":LinearColor=&MakeLinearColorFromVectorAndFloat(
     "Vector (RGB)"=&MakeVectorFromLinearColorRGB(Color=&ColorFromCurve(ColorCurve=DataInterface() {propertyValues {
         AlphaCurve {keys [<at 0 1>,<at 0.2388 1>,<at 1 0>]}
         BlueCurve {keys [<at 0 0>,<at 1 1>]}
         GreenCurve {keys [<at 0 1>,<at 1 0.1667>]}
         RedCurve {keys [<at 0 0>,<at 1 0>]}
        }}))
     "Float (Alpha)"=1
    )
    "Particles.Sparkly sphere DynParams (2)":Vector4f=&MakeVector4(
     X=&FloatFromCurve(FloatCurve=DataInterface() {propertyValues {Curve {keys [<at 0 0>,<at 1 0.02>]}}}, Scale Curve=0)
     Y=0
     Z=0
     W=0
    )
    "Particles.Sparkly sphere Size (2)":Vector2f=&Vector2DFromFloat(Value=&Multiply_Float(
      A=50
      B=&FloatFromCurve(
       FloatCurve=DataInterface() {propertyValues {Curve {keys [
           <at 0 0 RCIM_Cubic>
           <at 0.1 1.5>
           <at 0.175 1>
           <at 0.9 1>
           {arriveTangent=46.3816, interpMode=RCIM_Cubic, leaveTangent=-13.4952, tangentMode=RCTM_Break, time=1, value=0}
          ]}}}
       Scale Curve=1.25
      )
     ))
   ) {isSetParameters=true}
   /SetVariables(
    "Particles.Dark Glow Color (5)":LinearColor=&MakeLinearColorFromVectorAndFloat(
     "Vector (RGB)"=&MakeVectorFromLinearColorRGB(Color=<rgba 0.0452 0.0406 0.05 1>)
     "Float (Alpha)"=&FloatFromCurve(
      FloatCurve=DataInterface() {propertyValues {Curve {keys [
          {arriveTangent=3.7205, interpMode=RCIM_Cubic, leaveTangent=3.7205, tangentMode=RCTM_User, time=0, value=1}
          {interpMode=RCIM_Cubic, leaveTangent=0.002, time=0.8, value=0}
         ]}}}
      Scale Curve=0.5
     )
    )
    "Particles.Dark Glow size (5)":Vector2f=&Vector2DFromFloat(Value=&Multiply_Float(
      A=&FloatFromCurve(FloatCurve=DataInterface() {propertyValues {Curve {keys [
           {arriveTangent=0.7585, interpMode=RCIM_Cubic, leaveTangent=0.7585, tangentMode=RCTM_User, time=0, value=0.4}
           {interpMode=RCIM_Cubic, tangentMode=RCTM_User, time=1, value=1}
          ]}}})
      B=300
     ))
    "Particles.Hot Flare 2 SpriteRotation (3)":NiagaraFloat=&UniformRangedFloat(Minimum=-50, Maximum=50)
    "Particles.Hot Flare 2 color (3)":LinearColor=&MakeLinearColorFromVectorAndFloat(
     "Vector (RGB)"=&MakeVectorFromLinearColorRGB(Color=&ColorFromCurve(ColorCurve=DataInterface() {propertyValues {
         AlphaCurve {keys [<at 0 1 RCIM_Cubic>,<at 1 0 RCIM_Cubic>]}
         BlueCurve {keys [<at 0 2>, {interpMode=RCIM_Cubic, leaveTangent=1.3364, time=1, value=0.2016}]}
         GreenCurve {keys [<at 0 0.5272>, {interpMode=RCIM_Cubic, leaveTangent=-5.7627, time=1, value=1}]}
         RedCurve {keys [<at 0 0.0903>, {interpMode=RCIM_Cubic, leaveTangent=-13.364, time=1, value=0.012}]}
        }}))
     "Float (Alpha)"=&FloatFromCurve(
      FloatCurve=DataInterface() {propertyValues {Curve {keys [
          <at 0 0 RCIM_Cubic>
          {arriveTangent=0.3854, interpMode=RCIM_Cubic, leaveTangent=0.3854, tangentMode=RCTM_User, time=0.1, value=1}
          <at 1 0 RCIM_Cubic>
         ]}}}
      Scale Curve=0.5
     )
    )
    "Particles.Hot Flare 2 size (3)":Vector2f=&Vector2DFromFloat(Value=&Multiply_Float(
      A=50
      B=&FloatFromCurve(
       FloatCurve=DataInterface() {propertyValues {Curve {keys [
           <at 0 0 RCIM_Cubic>
           <at 0.1 1.5>
           <at 0.175 1>
           <at 0.9 1>
           {arriveTangent=46.3816, interpMode=RCIM_Cubic, leaveTangent=-13.4952, tangentMode=RCTM_Break, time=1, value=0}
          ]}}}
       Scale Curve=1.25
      )
     ))
   ) {isSetParameters=true}
   &CameraOffset(Camera Offset Amount=&MergeHelper_FloatSwitch(0=20, 1=50, 2=5, 3=1, 4=1, Index=&SpawnGroup(), Num=::NewEnumerator3))
   &FN_Light_Attributes(
    Color=Linked() {linkedVariable {name="Particles.Hot Flare 2 color (3)", type=LinearColor}}
    Create Color Attribute=true
    Create Exponent Attribute=true
    Create Radius Attribute=true
    Exponent:NiagaraFloat=&FloatFromCurve(FloatCurve=DataInterface() {propertyValues {Curve {keys [
         {interpMode=RCIM_Cubic, tangentMode=RCTM_User, time=0, value=1}
         {arriveTangent=-3.4524, interpMode=RCIM_Cubic, leaveTangent=-3.4524, tangentMode=RCTM_User, time=0.75, value=0}
        ]}}})
    Radius=&Multiply_Float(
     A=&FloatFromCurve(
      FloatCurve=DataInterface() {propertyValues {Curve {keys [
          {arriveTangent=2.3282, interpMode=RCIM_Cubic, leaveTangent=2.3282, tangentMode=RCTM_User, time=0, value=0}
          {interpMode=RCIM_Cubic, tangentMode=RCTM_User, time=1, value=1}
         ]}}}
      Scale Curve=150
     )
     B=0.5
    )
    Scale Color=&FloatFromCurve(
     FloatCurve=DataInterface() {propertyValues {Curve {keys [
         {interpMode=RCIM_Cubic, tangentMode=RCTM_User, time=0, value=1}
         {arriveTangent=-3.4524, interpMode=RCIM_Cubic, leaveTangent=-3.4524, tangentMode=RCTM_User, time=0.75, value=0}
        ]}}}
     Scale Curve=7
    )
   )
  ]
  rendererClasses [&NiagaraSpriteRendererProperties, &NiagaraLightRendererProperties]
  renderers [
   {rendererClass=&NiagaraSpriteRendererProperties}
   {rendererClass=&NiagaraSpriteRendererProperties, rendererIndex=1}
   {rendererClass=&NiagaraSpriteRendererProperties, rendererIndex=2}
   {rendererClass=&NiagaraSpriteRendererProperties, rendererIndex=3}
   {rendererClass=&NiagaraLightRendererProperties, rendererIndex=4}
   {rendererClass=&NiagaraSpriteRendererProperties, rendererIndex=5}
  ]
 }
 Glass_Shatter002 {
  emitterUpdateScript [
   &EmitterState(Life Cycle Mode=::NewEnumerator1, Loop Behavior=::NewEnumerator1, Loop Duration=1, UseLoopDelay=false)
   &SpawnBurst_Instantaneous(
    Spawn Count=&RandomRangeInt(Minimum=9, Maximum=15)
    Spawn Group=5
    Use Spawn Probability=false
    UseLoopCountLimit=false
   ) {module=SpawnBurst_Instantaneous004}
   &SpawnBurst_Instantaneous(
    Spawn Count=&RandomRangeInt(Minimum=9, Maximum=25)
    Spawn Group=4
    Use Spawn Probability=false
    UseLoopCountLimit=false
   ) {module=SpawnBurst_Instantaneous003}
   &SpawnBurst_Instantaneous(
    Spawn Count=&RandomRangeInt(Minimum=&RandomRangeInt(Minimum=9, Maximum=15), Maximum=6)
    Use Spawn Probability=false
    UseLoopCountLimit=false
   )
   &SpawnBurst_Instantaneous(Spawn Count=10, Spawn Group=1, Use Spawn Probability=false, UseLoopCountLimit=false) {module=SpawnBurst_Instantaneous001}
   &SpawnBurst_Instantaneous(
    Spawn Count=&RandomRangeInt(Minimum=9, Maximum=15)
    Spawn Group=3
    Use Spawn Probability=false
    UseLoopCountLimit=false
   ) {module=SpawnBurst_Instantaneous002}
  ]
  particleSpawnScript [
   /SetVariables(Particles.VisibilityTag=&SpawnGroup()) {isSetParameters=true}
   /SetVariables(
    Particles.Random Scale Size Float:NiagaraFloat=&RandomRangeFloat(Minimum=0.2, Maximum=0.8)
    Particles.Triangle Random Float:NiagaraFloat=&RandomRangeFloat(Minimum=0, Maximum=0.3)
   ) {isSetParameters=true}
   &InitializeParticle(
    Color=&MergeHelper_ColorSwitch(
     0=&RandomRangeLinearColor(Minimum=<rgba 0.6487 2 1.359 1>, Maximum=<rgba 0.605 2 1.1817 1>)
     1=<rgba 0.231 5 2.1335 1>
     2=&Multiply_LinearColorByFloat(Linear Color=<rgba 0.0172 1 0.51 1>, Float=&Lerp_Float(A=1, B=2, Alpha=$Particles.Triangle Random Float))
     3=?
     Index=&SpawnGroup()
     Num=::NewEnumerator1
    )
    Lifetime Max=0.6
    Lifetime Min=0.3
    Lifetime Mode=::NewEnumerator1
    Mesh Scale Mode=::NewEnumerator3
    Sprite Rotation Mode=::NewEnumerator2
    Sprite Size Mode=::NewEnumerator3
    Uniform Sprite Size=&MergeHelper_FloatSwitch(
     0=&RandomRangeFloat(Minimum=20, Maximum=50)
     1=&RandomRangeFloat(Minimum=10, Maximum=50)
     2=?
     3=?
     Index=&SpawnGroup()
     Num=::NewEnumerator0
    )
    UsePositionOffset=false
    Write Material Random=true
   )
   &ShapeLocation(MaskSpawnGroup=false, Override Seed=false, Rotation Axis=<0 0 1>, Rotation Mode=::NewEnumerator1, Sphere Radius=2, UseOwnerScale=true)
   &CameraOffset(Camera Offset Amount=&RandomRangeFloat(Minimum=2, Maximum=4))
   &AddVelocity(
    Module.Use Falloff Distance=false
    Override Seed=false
    Velocity Mode=::NewEnumerator1
    Velocity Speed=&FN_FrontEndFloatModifier(
     Gameplay=&MakeCustomFloatFromBool(
      Bool=&SetBoolByFloatComparison(A=&VectorLength(Vector=$User.OwnerLocalVelocity), B=150)
      True Float=200
      False Float=&RandomRangeFloat(Minimum=150, Maximum=350)
     )
     Front End=5
    )
   )
   &InitialMeshOrientation(Rotation=&RandomRangeVector(Minimum=<0 0 0>, Maximum=<180 180 180>), Use Rotation Vector=false)
   &AddVelocity(Velocity=&RandomRangeVector(Minimum=<250 250 -150>, Maximum=<-250 -250 350>)) {module=AddVelocity001}
  ]
  particleUpdateScript [
   &ParticleState()
   &GravityForce() {module=GravityForce001}
   &ScaleVelocity(Coordinate Space:ENiagaraCoordinateSpace=::Simulation, Velocity Scale=<3 3 3>)
   &Drag(
    Drag=&FloatFromCurve(
     FloatCurve=DataInterface() {propertyValues {Curve {
        defaultValue=1
        keys [
         {interpMode=RCIM_Cubic, tangentMode=RCTM_Break, time=0.4, value=17.5}
         <at 0.4149 17.5>
         <at 0.5149 500>
         {interpMode=RCIM_Cubic, tangentMode=RCTM_Break, time=1, value=500}
        ]
       }}}
     Scale Curve=3
    )
    Use Linear Drag=true
    Use Rotational Drag=true
   ) {module=Drag001}
   &AccelerationForce(Acceleration=<0 0 -2000>)
   &SolveForcesAndVelocity(Clamp Velocity=false, Limit Acceleration=false)
   &ScaleSpriteSizeBySpeed(
    Max Scale Factor=&MergeHelper_Vector2Switch(0=<1 1>, 1=<2 25>, 2=?, 3=?, 4=?, Index=0, Num=::NewEnumerator0)
    Velocity Threshold=100
   )
   &UpdateMeshOrientation(
    Rotation Rate=&FloatFromCurve(
     FloatCurve=DataInterface() {propertyValues {Curve {defaultValue=0, keys [{tangentMode=RCTM_User, time=0, value=1}, {tangentMode=RCTM_User, time=1, value=0}]}}}
     Scale Curve=&RandomRangeFloat(Minimum=0.5, Maximum=1)
    )
    Rotation Vector=&RandomRangeVector(Minimum=<0 0 0>, Maximum=<1 1 1>)
   )
   /SetVariables(
    Particles.Glitter sparkles:LinearColor=&MakeLinearColorFromVectorAndFloat(
     "Vector (RGB)"=&MakeVectorFromLinearColorRGB(Color=<rgba 0.231 5 2.1335 1>)
     "Float (Alpha)"=&FloatFromCurve(FloatCurve=DataInterface() {propertyValues {Curve {
         defaultValue=1
         keys [
          {arriveTangent=9.0725, interpMode=RCIM_Cubic, leaveTangent=9.0725, tangentMode=RCTM_User, time=0, value=0}
          {arriveTangent=-2.5249, interpMode=RCIM_Cubic, leaveTangent=-2.5249, tangentMode=RCTM_User, time=0.1408, value=0.9847}
          {arriveTangent=-0.0405, interpMode=RCIM_Cubic, leaveTangent=-0.0405, tangentMode=RCTM_User, time=1, value=0}
         ]
        }}})
    )
    Particles.Soft Sparkle Color:LinearColor=&MultiplyLinearColor(
     Linear Color=&ColorFromCurve(ColorCurve=DataInterface() {propertyValues {
        AlphaCurve {keys [<at 0 0>,<at 0.15 0.4 RCIM_Cubic>,<at 1 0 RCIM_Cubic>]}
        BlueCurve {keys [<at 0.1483 0.8796 RCIM_Cubic>,<at 1 0.51 RCIM_Cubic>]}
        GreenCurve {keys [<at 0.1483 0.9767 RCIM_Cubic>,<at 1 1 RCIM_Cubic>]}
        RedCurve {keys [<at 0.1483 1 RCIM_Cubic>,<at 1 0.0172 RCIM_Cubic>]}
       }})
     Scale Factor=&MakeFloat4FromVectorAndFloat("Vector (RGB)"=<1 1 1>, "Float (Alpha)"=$Particles.TimeOfDay)
    )
   ) {isSetParameters=true}
   &ScaleSpriteSize(Scale Factor=&MergeHelper_Vector2Switch(
     0=&Vector2DFromFloat(Value=&Multiply_Float(
       A=&FloatFromCurve(
        FloatCurve=DataInterface() {propertyValues {Curve {keys [
            {interpMode=RCIM_Cubic, tangentMode=RCTM_Break, time=0, value=0}
            {interpMode=RCIM_Cubic, tangentMode=RCTM_Break, time=0.1, value=1}
            {interpMode=RCIM_Cubic, tangentMode=RCTM_Break, time=0.2, value=0}
            {arriveTangent=5.8511, interpMode=RCIM_Cubic, leaveTangent=-6.2653, tangentMode=RCTM_Break, time=0.3023, value=0.9569}
            {interpMode=RCIM_Cubic, tangentMode=RCTM_Break, time=0.4029, value=0}
            {arriveTangent=7.3852, interpMode=RCIM_Cubic, leaveTangent=-4.8986, tangentMode=RCTM_Break, time=0.5198, value=0.7267}
            {interpMode=RCIM_Cubic, tangentMode=RCTM_Break, time=0.6539, value=0}
            {interpMode=RCIM_Cubic, leaveTangent=-5.1375, tangentMode=RCTM_Break, time=0.7904, value=0.4154}
            {interpMode=RCIM_Cubic, tangentMode=RCTM_Break, time=0.9977, value=0}
           ]}}}
        Scale Curve=1.5
       )
       B=&Lerp_Float(
        A=1
        B=&FloatFromCurve(
         FloatCurve=DataInterface() {propertyValues {Curve {keys [
             {interpMode=RCIM_Cubic, tangentMode=RCTM_Break, time=0, value=0}
             {interpMode=RCIM_Cubic, tangentMode=RCTM_Break, time=0.1, value=1}
             {interpMode=RCIM_Cubic, tangentMode=RCTM_Break, time=0.2, value=0}
             {arriveTangent=5.8511, interpMode=RCIM_Cubic, leaveTangent=-6.2653, tangentMode=RCTM_Break, time=0.3023, value=0.9569}
             {interpMode=RCIM_Cubic, tangentMode=RCTM_Break, time=0.4029, value=0}
             {arriveTangent=7.3852, interpMode=RCIM_Cubic, leaveTangent=-4.8986, tangentMode=RCTM_Break, time=0.5198, value=0.7267}
             {interpMode=RCIM_Cubic, tangentMode=RCTM_Break, time=0.6539, value=0}
             {interpMode=RCIM_Cubic, leaveTangent=-5.1375, tangentMode=RCTM_Break, time=0.7904, value=0.4154}
             {interpMode=RCIM_Cubic, tangentMode=RCTM_Break, time=0.9977, value=0}
            ]}}}
         CurveIndex=&Multiply_Float(A=$Particles.NormalizedAge, B=$Particles.Random Scale Size Float)
         Scale Curve=1.5
        )
        Alpha=0.5
       )
      ))
     1=&Vector2DFromFloat(Value=&Multiply_Float(
       A=&FloatFromCurve(
        FloatCurve=DataInterface() {propertyValues {Curve {keys [
            {interpMode=RCIM_Cubic, tangentMode=RCTM_Break, time=0, value=0}
            {interpMode=RCIM_Cubic, tangentMode=RCTM_Break, time=0.1, value=1}
            {interpMode=RCIM_Cubic, tangentMode=RCTM_Break, time=0.2, value=0}
            {arriveTangent=5.8511, interpMode=RCIM_Cubic, leaveTangent=-6.2653, tangentMode=RCTM_Break, time=0.3023, value=0.9569}
            {interpMode=RCIM_Cubic, tangentMode=RCTM_Break, time=0.4029, value=0}
            {arriveTangent=7.3852, interpMode=RCIM_Cubic, leaveTangent=-4.8986, tangentMode=RCTM_Break, time=0.5198, value=0.7267}
            {interpMode=RCIM_Cubic, tangentMode=RCTM_Break, time=0.6539, value=0}
            {interpMode=RCIM_Cubic, leaveTangent=-5.1375, tangentMode=RCTM_Break, time=0.7904, value=0.4154}
            {interpMode=RCIM_Cubic, tangentMode=RCTM_Break, time=0.9977, value=0}
           ]}}}
        Scale Curve=0.5
       )
       B=&Lerp_Float(
        A=1
        B=&FloatFromCurve(
         FloatCurve=DataInterface() {propertyValues {Curve {keys [
             {interpMode=RCIM_Cubic, tangentMode=RCTM_Break, time=0, value=0}
             {interpMode=RCIM_Cubic, tangentMode=RCTM_Break, time=0.1, value=1}
             {interpMode=RCIM_Cubic, tangentMode=RCTM_Break, time=0.2, value=0}
             {arriveTangent=5.8511, interpMode=RCIM_Cubic, leaveTangent=-6.2653, tangentMode=RCTM_Break, time=0.3023, value=0.9569}
             {interpMode=RCIM_Cubic, tangentMode=RCTM_Break, time=0.4029, value=0}
             {arriveTangent=7.3852, interpMode=RCIM_Cubic, leaveTangent=-4.8986, tangentMode=RCTM_Break, time=0.5198, value=0.7267}
             {interpMode=RCIM_Cubic, tangentMode=RCTM_Break, time=0.6539, value=0}
             {interpMode=RCIM_Cubic, leaveTangent=-5.1375, tangentMode=RCTM_Break, time=0.7904, value=0.4154}
             {interpMode=RCIM_Cubic, tangentMode=RCTM_Break, time=0.9977, value=0}
            ]}}}
         CurveIndex=&Multiply_Float(A=$Particles.NormalizedAge, B=$Particles.Random Scale Size Float)
        )
        Alpha=0.5
       )
      ))
     2=&Vector2DFromCurve(Vector2Curve=DataInterface() {propertyValues {XCurve {keys [<at 0 1>,<at 1 0>]}, YCurve {keys [<at 0 1>,<at 1 0>]}}}, ScaleCurve=<1 1>)
     3=?
     4=?
     Index=$Particles.VisibilityTag
     Num=::NewEnumerator0
    )) {module=ScaleSpriteSize001}
   &ScaleColor(
    Scale Alpha=&Multiply_Float(
     A=&MergeHelper_FloatSwitch(
      0=&FloatFromCurve(FloatCurve=DataInterface() {propertyValues {Curve {keys [<at 0 0>,<at 0.1 1>,<at 0.5 1>,<at 0.85 0>]}}})
      1=0
      2=&Multiply_Float(
       A=1
       B=&FloatFromCurve(FloatCurve=DataInterface() {propertyValues {Curve {keys [<at 0.5 1>,<at 1 0.5>]}}})
      )
      3=?
      Index=$Particles.VisibilityTag
      Num=::NewEnumerator1
     )
     B=&Waveform(
      "Waveform [1]"=::NewEnumerator0
      "[1] Frequency"=&RandomRangeFloat(Minimum=3, Maximum=6)
      "[1] Amplitude Min/Max"=<0 1>
     )
    )
    Scale RGB=&MergeHelper_VectorSwitch(
     0=<1 1 1>
     1=<5 5 5>
     2=&MakeVectorFromLinearColorRGB(Color=&ColorFromCurve(ColorCurve=DataInterface() {propertyValues {
         AlphaCurve {keys [<at 0 0 RCIM_Cubic>,<at 0.1725 1>]}
         BlueCurve {keys [<at 0 2.4584 RCIM_Cubic>,<at 0.2942 0.9291>]}
         GreenCurve {keys [<at 0 1.5446 RCIM_Cubic>,<at 0.2942 0.8885>]}
         RedCurve {keys [<at 0 3 RCIM_Cubic>,<at 0.2942 1>]}
        }}))
     3=?
     Index=$Particles.VisibilityTag
     Num=::NewEnumerator1
    )
    ScaleA=true
    ScaleRGB=true
   )
   /SetVariables(
    Particles.Glass 2 scale:Vector3f=&VectorFromFloat(Value=&FloatFromCurve(
      FloatCurve=DataInterface() {propertyValues {Curve {keys [
          {arriveTangent=26.6536, interpMode=RCIM_Cubic, leaveTangent=26.6536, leaveTangentWeight=0.25, tangentMode=RCTM_User, time=0, value=0}
          <at 0.1 1>
          {interpMode=RCIM_Cubic, leaveTangent=0.239, tangentMode=RCTM_Break, time=0.6, value=1}
         ]}}}
      Scale Curve=&RandomRangeFloat(Minimum=0.2, Maximum=0.5)
     ))
    Particles.glass 1 scale:Vector3f=&VectorFromFloat(Value=&FloatFromCurve(
      FloatCurve=DataInterface() {propertyValues {Curve {keys [
          {arriveTangent=26.6536, interpMode=RCIM_Cubic, leaveTangent=26.6536, leaveTangentWeight=0.2833, tangentMode=RCTM_User, time=0, value=0}
          <at 0.1 1>
          {arriveTangentWeight=0.1333, interpMode=RCIM_Cubic, leaveTangent=0.239, tangentMode=RCTM_Break, time=0.55, value=1}
         ]}}}
      Scale Curve=&RandomRangeFloat(Minimum=0.2, Maximum=0.5)
     ))
    Particles.glass 3 scale:Vector3f=&VectorFromFloat(Value=&FloatFromCurve(
      FloatCurve=DataInterface() {propertyValues {Curve {keys [
          {arriveTangent=26.6536, interpMode=RCIM_Cubic, leaveTangent=26.6536, leaveTangentWeight=0.25, tangentMode=RCTM_User, time=0, value=0}
          <at 0.1 1>
          {arriveTangentWeight=0.1333, interpMode=RCIM_Cubic, leaveTangent=0.239, tangentMode=RCTM_Break, time=0.5, value=1}
         ]}}}
      Scale Curve=&RandomRangeFloat(Minimum=0.2, Maximum=0.55)
     ))
    Particles.glass color:LinearColor=&MakeLinearColorFromVectorAndFloat(
     "Vector (RGB)"=&MakeVectorFromLinearColorRGB(Color=<rgba 0.5719 1 0.7866 1>)
     "Float (Alpha)"=&FloatFromCurve(
      FloatCurve=DataInterface() {propertyValues {Curve {keys [
          <at 0 1>
          {interpMode=RCIM_Cubic, tangentMode=RCTM_Break, time=0.2, value=1}
          {interpMode=RCIM_Cubic, tangentMode=RCTM_Break, time=0.6, value=0}
         ]}}}
      Scale Curve=&FNTimeOfDayScalar(Night Scale=3, Morning Scale=3, Day Scale=2, Evening Scale=2)
     )
    )
    Particles.test Scale:Vector3f=&VectorFromFloat(Value=&FloatFromCurve(
      FloatCurve=DataInterface() {propertyValues {Curve {keys [
          {arriveTangent=26.6536, interpMode=RCIM_Cubic, leaveTangent=26.6536, leaveTangentWeight=0.25, tangentMode=RCTM_User, time=0, value=0}
          <at 0.1 1>
          <at 0.75 1>
          <at 0.85 0>
         ]}}}
      Scale Curve=&RandomRangeFloat(Minimum=0.2, Maximum=1.3)
     ))
   ) {isSetParameters=true}
   &ScaleMeshSize(Scale Factor=&VectorFromFloat(Value=&FloatFromCurve(
      FloatCurve=DataInterface() {propertyValues {Curve {keys [{arriveTangent=14.45, interpMode=RCIM_Cubic, time=0, value=0}, {arriveTangent=2.5, leaveTangent=2.5, time=0.3, value=1}, <at 0.8 0>]}}}
      Scale Curve=5
     )))
   /SetVariables(Particles.TimeOfDay:NiagaraFloat=&FNTimeOfDayScalar(Night Scale=0.5, Morning Scale=0.3, Evening Scale=0.3)) {isSetParameters=true}
  ]
  rendererClasses [&NiagaraSpriteRendererProperties, &NiagaraMeshRendererProperties]
  renderers [
   {rendererClass=&NiagaraSpriteRendererProperties}
   {rendererClass=&NiagaraSpriteRendererProperties, rendererIndex=1}
   {rendererClass=&NiagaraMeshRendererProperties, rendererIndex=2}
   {rendererClass=&NiagaraMeshRendererProperties, rendererIndex=3}
   {rendererClass=&NiagaraMeshRendererProperties, rendererIndex=4}
   {rendererClass=&NiagaraSpriteRendererProperties, rendererIndex=5}
  ]
 }
 Spark_GPU001 {
  simTarget=GPUComputeSim
  emitterUpdateScript [
   &EmitterState(Life Cycle Mode=::NewEnumerator1, Loop Behavior=::NewEnumerator1, Loop Delay=0.1, Loop Duration=1, UseLoopDelay=false)
   &SpawnBurst_Instantaneous(Spawn Count=&UniformRangedInt(Minimum=64, Maximum=80), Spawn Time=0.05, Use Spawn Probability=false)
  ]
  particleSpawnScript [
   &InitializeParticle(
    Color=<rgba 0.0163 0.8534 1 1>
    Lifetime Max=1.5
    Lifetime Min=0.5
    Lifetime Mode=::NewEnumerator1
    Position Offset=<0 0 1>
    Sprite Size Max=<5 7>
    Sprite Size Min=<2 3>
    Sprite Size Mode=::NewEnumerator2
    UsePositionOffset=true
    Write Material Random=true
   ) {module=InitializeParticle001}
   &SphereLocation(
    MaskSpawnGroup=false
    Offset=<0 0 -5>
    Override Seed=false
    Sphere Radius=3
    Surface Expansion Mode=::NewEnumerator1
    Surface Only=true
    Surface Only Band Thickness=1
    UseNonUniformScale=false
    UseOffset=true
    UseSphereOrientation=false
   ) {module=SphereLocation001}
   &AddVelocityFromPoint(
    Module.Use Falloff Distance=false
    Origin Offset=<0 0 -5>
    Velocity Strength=&Multiply_Float(A=&RandomRangeFloat(Minimum=0.5, Maximum=1), B=1350)
   )
  ]
  particleUpdateScript [
   &ParticleState()
   &CurlNoiseForce(Noise Frequency=8, Noise Strength=1500, Pan Noise=true, Pan Noise Field=<0 0.2 0.5>)
   &Drag(Drag=&UniformRangedFloat(Minimum=5, Maximum=20), Use Linear Drag=true, Use Rotational Drag=false)
   &SolveForcesAndVelocity(Clamp Velocity=false, Limit Acceleration=false)
   &ScaleSpriteSize(Scale Factor=&Vector2DFromCurve(
     Vector2Curve=DataInterface() {propertyValues {
       XCurve {keys [<at 0 1 RCIM_Cubic>, {arriveTangent=-0.5084, interpMode=RCIM_Cubic, leaveTangent=-0.5084, tangentMode=RCTM_User, time=1, value=0}]}
       YCurve {keys [{arriveTangent=-3.0155, interpMode=RCIM_Cubic, leaveTangent=-3.0155, tangentMode=RCTM_User, time=0, value=1}, <at 1 0 RCIM_Cubic>]}
      }}
     ScaleCurve=&Vector2DFromFloat(Value=&Lerp_Float(
       A=&FloatFromCurve(
        FloatCurve=DataInterface() {propertyValues {Curve {keys [
            {interpMode=RCIM_Cubic, tangentMode=RCTM_Break, time=0, value=0.5}
            {arriveTangent=0.001, interpMode=RCIM_Cubic, tangentMode=RCTM_Break, time=1000, value=1}
           ]}}}
        CurveIndex=$Engine.Owner.LODDistance
       )
       B=1
       Alpha=1
      ))
    ))
   &ScaleSpriteSizeBySpeed(
    Max Scale Factor=&RandomRangeVector2D(Minimum=<0.1 0.2>, Maximum=<0.5 4>)
    Min Scale Factor=<0.1 0.3>
    Velocity Threshold=500
   )
   &ScaleColor(
    Scale Alpha=&FloatFromCurve(
     FloatCurve=DataInterface() {propertyValues {Curve {keys [
         {tangentMode=RCTM_User, time=0, value=1}
         {arriveTangent=-0.0405, interpMode=RCIM_Cubic, leaveTangent=-0.0405, tangentMode=RCTM_User, time=0.7009, value=1}
         {arriveTangent=-8.0089, interpMode=RCIM_Cubic, leaveTangent=-8.0089, tangentMode=RCTM_User, time=1, value=0}
        ]}}}
     Scale Curve=&FNTimeOfDayScalar(Night Scale=10, Morning Scale=7, Day Scale=7, Evening Scale=7)
    )
    Scale RGB=<3 3 3>
    ScaleA=true
    ScaleRGB=true
   )
  ]
  rendererClasses [&NiagaraSpriteRendererProperties]
  renderers [{rendererClass=&NiagaraSpriteRendererProperties}]
 }
 Star_impact_ground {
  emitterUpdateScript [
   &EmitterState(Life Cycle Mode=::NewEnumerator1, Loop Behavior=::NewEnumerator1, Loop Duration=1.5, UseLoopDelay=false)
   &SpawnBurst_Instantaneous(Spawn Group=4, Use Spawn Probability=false, UseLoopCountLimit=false) {module=SpawnBurst_Instantaneous002}
   &SpawnBurst_Instantaneous(Spawn Group=3, Spawn Time=0.01, Use Spawn Probability=false, UseLoopCountLimit=false) {module=SpawnBurst_Instantaneous004}
   &SpawnBurst_Instantaneous(Spawn Time=0.01, Use Spawn Probability=false, UseLoopCountLimit=false) {module=SpawnBurst_Instantaneous001}
   &SpawnBurst_Instantaneous(Spawn Group=1, Spawn Time=0.03, Use Spawn Probability=false, UseLoopCountLimit=false)
  ]
  particleSpawnScript [
   /SetVariables(Particles.VisibilityTag=&SpawnGroup()) {isSetParameters=true}
   &InitializeParticle(
    Color Mode=::NewEnumerator0
    Lifetime=&MergeHelper_FloatSwitch(0=1.5, 1=1.5, 2=1.5, 3=1.5, 4=0.5, Index=&SpawnGroup(), Num=::NewEnumerator3)
    Mesh Scale Mode=::NewEnumerator3
    Mesh Uniform Scale=3.2
    Position Offset=&MergeHelper_VectorSwitch(0=<0 0 0>, 1=<0 0 0>, 2=<0 0 0>, 3=<0 0 0>, 4=<0 0 3>, Index=&SpawnGroup(), Num=::NewEnumerator3)
    UsePositionOffset=false
    Write Material Random=true
   )
   /SetVariables(
    Particles.1 dynparam:Vector4f=&MakeVector4(X=0, Y=&RandomRangeFloat(Minimum=0, Maximum=4), Z=0, W=2.5)
    Particles.SpriteAlignment=&TransformVector(Vector=<0 0 0>, Destination Space=::World)
   ) {isSetParameters=true}
   &SkeletalMeshLocation(
    Bone Sampling Mode=::NewEnumerator4
    MaskSpawnGroup=false
    Orientation Sampling=::NewEnumerator1
    Position Sampling=::NewEnumerator2
    Sampled Position Offset=<0 55 0>
    Sampled Position Offset Coordinate Space:ENiagaraSkelMeshTransforms=::NewEnumerator2
    Skeletal Mesh=$User.OwnerSkeleton
   ) {module=SkeletalMeshLocation001}
   &CameraOffset(Camera Offset Amount=&MergeHelper_FloatSwitch(0=0, 1=3, 2=30, 3=0, 4=0, Index=&SpawnGroup(), Num=::NewEnumerator3))
   &SpriteFacingAndAlignment(Alignment Coordinate Space=::Local, Sprite Facing=<0 0 1>, WriteAlignment=true, WriteFacing=true)
  ]
  particleUpdateScript [
   /SetVariables(
    Particles.Glass DynParam:Vector4f=&MakeVector4(
     X=&FloatFromCurve(FloatCurve=DataInterface() {propertyValues {Curve {keys [{arriveTangent=-2.1015, interpMode=RCIM_Cubic, leaveTangent=-2.1015, tangentMode=RCTM_User, time=0, value=0.5}, <at 1 -0.15>]}}})
     Y=&FloatFromCurve(FloatCurve=DataInterface() {propertyValues {Curve {keys [{arriveTangent=-0.3464, interpMode=RCIM_Cubic, leaveTangent=-0.3464, tangentMode=RCTM_User, time=0, value=1}, <at 1 0.9>]}}})
     Z=&FloatFromCurve(FloatCurve=DataInterface() {propertyValues {Curve {keys [
          {interpMode=RCIM_Cubic, tangentMode=RCTM_Break, time=0, value=0}
          {arriveTangent=0.4478, interpMode=RCIM_Cubic, leaveTangent=0.4478, tangentMode=RCTM_User, time=0.15, value=1}
          {interpMode=RCIM_Cubic, tangentMode=RCTM_Break, time=1, value=1.2}
         ]}}})
     W=0
    )
    Particles.Glow Dynamic params:Vector4f=&MakeVector4(X=2, Y=0, Z=0, W=0)
    Particles.Main Star DynParams:Vector4f=&MakeVector4(
     X=&FloatFromCurve(FloatCurve=DataInterface() {propertyValues {Curve {keys [
          <at 0 0.05>
          {arriveTangent=2.6196, interpMode=RCIM_Cubic, leaveTangent=2.6196, tangentMode=RCTM_User, time=0.5, value=0.05}
          {interpMode=RCIM_Cubic, tangentMode=RCTM_User, time=0.9, value=1}
         ]}}})
     Y=&RandomRangeFloat(Minimum=0, Maximum=1)
     Z=0
     W=0
    )
   ) {isSetParameters=true}
   /SetVariables(
    Particles.Star Color:LinearColor=&MakeLinearColorFromVectorAndFloat(
     "Vector (RGB)"=&MakeVectorFromLinearColorRGB(Color=&ColorFromCurve(ColorCurve=DataInterface() {propertyValues {
         AlphaCurve {keys [<at 0 1 RCIM_Cubic>,<at 1 0 RCIM_Cubic>]}
         BlueCurve {keys [<at 0 0.9436 RCIM_Cubic>,<at 1 1 RCIM_Cubic>]}
         GreenCurve {keys [<at 0 1 RCIM_Cubic>,<at 1 1 RCIM_Cubic>]}
         RedCurve {keys [<at 0 0.7635 RCIM_Cubic>,<at 1 1 RCIM_Cubic>]}
        }}))
     "Float (Alpha)"=&FloatFromCurve(FloatCurve=DataInterface() {propertyValues {Curve {keys [<at 0 1>,<at 1 0>]}}})
    )
    Particles.Star Glow Color:LinearColor=&MakeLinearColorFromVectorAndFloat(
     "Vector (RGB)"=&MakeVectorFromLinearColorRGB(Color=&ColorFromCurve(ColorCurve=DataInterface() {propertyValues {
         AlphaCurve {keys [<at 0 1 RCIM_Cubic>,<at 0.8144 0 RCIM_Cubic>]}
         BlueCurve {keys [<at 0 0.7613 RCIM_Cubic>,<at 1 0.8107 RCIM_Cubic>]}
         GreenCurve {keys [<at 0 1 RCIM_Cubic>,<at 1 1 RCIM_Cubic>]}
         RedCurve {keys [<at 0 0 RCIM_Cubic>,<at 1 0 RCIM_Cubic>]}
        }}))
     "Float (Alpha)"=&FloatFromCurve(FloatCurve=DataInterface() {propertyValues {Curve {keys [<at 0 1>,<at 0.7 0>]}}}, Scale Curve=0.5)
    )
    Particles.Star Glow Color 2:LinearColor=&MakeLinearColorFromVectorAndFloat(
     "Vector (RGB)"=&MakeVectorFromLinearColorRGB(Color=&ColorFromCurve(ColorCurve=DataInterface() {propertyValues {
         AlphaCurve {keys [<at 0 1 RCIM_Cubic>,<at 1 0 RCIM_Cubic>]}
         BlueCurve {keys [<at 0 1 RCIM_Cubic>,<at 1 1 RCIM_Cubic>]}
         GreenCurve {keys [<at 0 1 RCIM_Cubic>,<at 1 1 RCIM_Cubic>]}
         RedCurve {keys [<at 0 1 RCIM_Cubic>,<at 1 1 RCIM_Cubic>]}
        }}))
     "Float (Alpha)"=&FloatFromCurve(FloatCurve=DataInterface() {propertyValues {Curve {keys [<at 0 1>,<at 0.7 0>]}}}, Scale Curve=5)
    )
    Particles.Star Glow DyNParam:Vector4f=&MakeVector4(
     X=&FloatFromCurve(FloatCurve=DataInterface() {propertyValues {Curve {keys [
          {arriveTangent=-0.0494, interpMode=RCIM_Cubic, leaveTangent=-0.0494, tangentMode=RCTM_User, time=0.5, value=0}
          {arriveTangent=1.5554, interpMode=RCIM_Cubic, leaveTangent=1.5554, tangentMode=RCTM_User, time=1, value=1}
         ]}}})
     Y=0
     Z=&Multiply_Float(
      A=&RandomRangeFloat(Minimum=0.01, Maximum=0.04)
      B=&Lerp_Float(A=1.5, B=2.5, Alpha=0.5)
     )
     W=0
    )
    Particles.Star Glow Position:NiagaraPosition=&AddVectorToPosition(Position=<0 0 2>, Vector=<0 0 0>)
    Particles.Star Glow Position 2:NiagaraPosition=&AddVectorToPosition(Position=<0 0 5>, Vector=<0 0 1>)
    Particles.Star Glow Size:Vector2f=&Multiply_Vector2D_ByFloat(
     Float=&FloatFromCurve(FloatCurve=DataInterface() {propertyValues {Curve {
         defaultValue=0
         keys [
          {interpMode=RCIM_Cubic, tangentMode=RCTM_Break, time=0, value=0}
          {interpMode=RCIM_Cubic, leaveTangent=-0.1735, tangentMode=RCTM_Break, time=0.15, value=1.75}
          {arriveTangent=-0.177, interpMode=RCIM_Cubic, leaveTangent=-0.177, tangentMode=RCTM_User, time=0.3, value=1}
          {arriveTangent=-0.15, interpMode=RCIM_Cubic, leaveTangent=-0.15, tangentMode=RCTM_User, time=0.9574, value=1}
         ]
        }}})
     Vector2D=<168 168>
    )
    Particles.Star Glow size 2:Vector2f=&Multiply_Vector2D_ByFloat(
     Float=&FloatFromCurve(FloatCurve=DataInterface() {propertyValues {Curve {
         defaultValue=0
         keys [
          {interpMode=RCIM_Cubic, tangentMode=RCTM_Break, time=0, value=0}
          {interpMode=RCIM_Cubic, leaveTangent=-0.1735, tangentMode=RCTM_Break, time=0.15, value=1.75}
          {arriveTangent=-0.177, interpMode=RCIM_Cubic, leaveTangent=-0.177, tangentMode=RCTM_User, time=0.3, value=1}
          {arriveTangent=-0.15, interpMode=RCIM_Cubic, leaveTangent=-0.15, tangentMode=RCTM_User, time=0.9574, value=1}
         ]
        }}})
     Vector2D=<180 180>
    )
    Particles.Star Position:NiagaraPosition=&AddVectorToPosition(Position=<0 0 2>, Vector=<0 0 0>)
    Particles.Star Sprite Size:Vector2f=&Multiply_Vector2D_ByFloat(
     Float=&FloatFromCurve(FloatCurve=DataInterface() {propertyValues {Curve {
         defaultValue=0
         keys [
          {interpMode=RCIM_Cubic, tangentMode=RCTM_Break, time=0, value=0}
          {interpMode=RCIM_Cubic, leaveTangent=-0.1735, tangentMode=RCTM_Break, time=0.15, value=1.75}
          {arriveTangent=-0.177, interpMode=RCIM_Cubic, leaveTangent=-0.177, tangentMode=RCTM_User, time=0.3, value=1}
          {arriveTangent=-0.15, interpMode=RCIM_Cubic, leaveTangent=-0.15, tangentMode=RCTM_User, time=0.9574, value=1}
         ]
        }}})
     Vector2D=<156 156>
    )
    Particles.Star SpriteAlignment:Vector3f=<1 0 1>
    Particles.star DynParams:Vector4f=&MakeVector4(
     X=&FloatFromCurve(FloatCurve=DataInterface() {propertyValues {Curve {keys [
          <at 0 0>
          <at 0.1 0.04>
          {interpMode=RCIM_Cubic, tangentMode=RCTM_User, time=0.4, value=0.04}
          {interpMode=RCIM_Cubic, tangentMode=RCTM_User, time=1, value=1}
         ]}}})
     Y=&RandomRangeFloat(Minimum=0, Maximum=1)
     Z=0
     W=0
    )
   ) {isSetParameters=true}
   /SetVariables(
    Particles.Ground Glass Linear Color:LinearColor=&MakeLinearColorFromVectorAndFloat(
     "Vector (RGB)"=&MakeVectorFromLinearColorRGB(Color=&ColorFromCurve(ColorCurve=DataInterface() {propertyValues {
         AlphaCurve {keys [<at 0 1>,<at 0.3531 0.5267>,<at 1 0>]}
         BlueCurve {keys [<at 0 0>,<at 1 3>]}
         GreenCurve {keys [<at 0 3>,<at 1 0.5>]}
         RedCurve {keys [<at 0 0>,<at 1 0>]}
        }}))
     "Float (Alpha)"=&FloatFromCurve(
      FloatCurve=DataInterface() {propertyValues {Curve {keys [{interpMode=RCIM_Cubic, tangentMode=RCTM_User, time=0.5, value=1}, {interpMode=RCIM_Cubic, tangentMode=RCTM_User, time=0.95, value=0}]}}}
      Scale Curve=&FN_FrontEndFloatModifier(Gameplay=&FNTimeOfDayScalar(Night Scale=3, Morning Scale=1, Evening Scale=1), Front End=0.5)
     )
    )
    Particles.Ground glass scale:Vector3f=&VectorFromFloat(Value=&FloatFromCurve(
      FloatCurve=DataInterface() {propertyValues {Curve {keys [
          {interpMode=RCIM_Cubic, tangentMode=RCTM_Break, time=0, value=0.2}
          {arriveTangent=2.9365, interpMode=RCIM_Cubic, leaveTangent=2.9365, tangentMode=RCTM_User, time=0.1, value=1}
          {interpMode=RCIM_Cubic, tangentMode=RCTM_Break, time=1, value=2}
         ]}}}
      Scale Curve=1.5
     ))
   ) {isSetParameters=true}
   &Drag(
    Drag=&MergeHelper_FloatSwitch(
     0=0
     1=0
     2=0
     3=0
     4=&FloatFromCurve(
      FloatCurve=DataInterface() {propertyValues {Curve {defaultValue=1, keys [{tangentMode=RCTM_User, time=0, value=0}, {tangentMode=RCTM_User, time=1, value=1}]}}}
      Scale Curve=60
     )
     Index=&SpawnGroup()
     Num=::NewEnumerator3
    )
    Use Linear Drag=true
    Use Rotational Drag=true
   )
   &SolveForcesAndVelocity(Clamp Velocity=false, Limit Acceleration=false)
   &Collision(
    Age Colliding Particles=false
    KillOnCollision=false
    Module.Control Rest State Via Surface Normal=false
    Module.EnableMaxCollisionCount=false
    Module.Randomize Collision Normal Vector=false
    Restitution=0
   )
   &ParticleState()
   &self:Slope_Position_01(
    Apply Position=true
    Apply Rotation=true
    Height Adjustment/Offset=1
    Slope Position:NiagaraPosition=$Particles.SkeletalMeshLocation001.SampledPosition
    Trace Length=120
    Trace Origin Height=85
    bIsFrontEnd:NiagaraBool=$User.bIsFrontEndPreview
   ) {module=Slope_Position_01}
   &BlendSpace("Position Blend (World <-> Local)":NiagaraFloat=&FloatFromCurve(
     FloatCurve=DataInterface() {propertyValues {Curve {
        defaultValue=0
        keys [{interpMode=RCIM_Cubic, tangentMode=RCTM_User, time=0, value=1}, {interpMode=RCIM_Cubic, tangentMode=RCTM_User, time=0.15, value=0}]
       }}}
     Translation Contribution:NiagaraFloat=2
     Rotation Contribution:NiagaraFloat=0
    ))
  ]
  rendererClasses [&NiagaraSpriteRendererProperties, &NiagaraMeshRendererProperties]
  renderers [
   {rendererClass=&NiagaraSpriteRendererProperties}
   {rendererClass=&NiagaraSpriteRendererProperties, rendererIndex=1}
   {rendererClass=&NiagaraSpriteRendererProperties, rendererIndex=2}
   {rendererClass=&NiagaraMeshRendererProperties, rendererIndex=3}
  ]
 }
 SwirlSmokeHead {
  emitterUpdateScript [
   &EmitterState(Life Cycle Mode=::NewEnumerator1, Loop Behavior=::NewEnumerator1, Loop Duration=0.8, UseLoopDelay=false)
   &SpawnBurst_Instantaneous(Spawn Count=12, Use Spawn Probability=false, UseLoopCountLimit=false)
  ]
  particleSpawnScript [
   &PartitionParticles(Number Of Partitions=5)
   &InitializeParticle(
    Color=&SelectLinearColorFromArray(
     Array Sampling Mode=::NewEnumerator1
     Direct Array Index=$Particles.PartitionParticles.Partition
     Color Selection Array=DataInterface() {propertyValues {ColorData [
        <rgba 1 0.5469 0.7601 1>
        <rgba 0.3134 3 0.06 1>
        <rgba 0.1199 1 0 1>
        <rgba 0.1562 0.3204 1 1>
        <rgba 1.1568 2 0.0346 1>
        <rgba 0.5365 0.1878 1 1>
       ]}}
     Random Seed=?
    )
    Lifetime Max=0.65
    Lifetime Min=0.35
    Lifetime Mode=::NewEnumerator1
    Sprite Rotation Mode=::NewEnumerator2
    Sprite Size Mode=::NewEnumerator1
    Uniform Sprite Size Max=150
    Uniform Sprite Size Min=75
    UsePositionOffset=false
    Write Material Random=true
   ) {module=InitializeParticle001}
   &SystemLocation(MaskSpawnGroup=false, Offset=<0 0 9>)
   &CameraOffset(Camera Offset Amount=&UniformRangedFloat(Minimum=4, Maximum=10))
   /SetVariables(Particles.SpriteFacing=<0.05 0.05 1>) {isSetParameters=true}
   &AddVelocity(Velocity=&RandomRangeVector(Minimum=<-150 -150 0>, Maximum=<150 150 1>)) {module=AddVelocity001}
  ]
  particleUpdateScript [
   &Drag(Drag=0.5, Use Linear Drag=true, Use Rotational Drag=false)
   &GravityForce(Gravity=<0 0 0>)
   &SolveForcesAndVelocity(Clamp Velocity=false, Limit Acceleration=false)
   &ParticleState()
   &ScaleSpriteSize()
   &SpriteRotationRate(Rotation Rate=$Particles.InitialRotationRate)
   &ScaleColor(
    Scale Alpha=&FloatFromCurve(
     FloatCurve=DataInterface() {propertyValues {Curve {keys [
         {arriveTangent=4.4407, interpMode=RCIM_Cubic, leaveTangent=4.4407, tangentMode=RCTM_User, time=0, value=0}
         {arriveTangent=0.0345, interpMode=RCIM_Cubic, leaveTangent=0.0345, tangentMode=RCTM_User, time=0.2332, value=0.3324}
         {arriveTangent=-1.3404, interpMode=RCIM_Cubic, leaveTangent=-1.3404, tangentMode=RCTM_User, time=1, value=0}
        ]}}}
     Scale Curve=&FNTimeOfDayScalar(Night Scale=0.15, Morning Scale=0.08, Day Scale=0.1, Evening Scale=0.08)
    )
    ScaleA=true
    ScaleRGB=false
   )
   &DynamicMaterialParameters(
    Index 0 Param 1=&FloatFromCurve(FloatCurve=DataInterface() {propertyValues {Curve {keys [<at 0.3 1 RCIM_Cubic>,<at 0.6 -0.1>]}}}, CurveIndex=$Particles.NormalizedLife)
    Index 0 Param 2=&FloatFromCurve(
     FloatCurve=DataInterface() {propertyValues {Curve {keys [<at 0 1 RCIM_Cubic>,<at 0.8 8.8 RCIM_Cubic>]}}}
     CurveIndex=$Particles.NormalizedLife
    )
    Index 0 Param 4=&FloatFromCurve(FloatCurve=DataInterface() {propertyValues {Curve {keys [<at 0.3 0>,<at 0.5 0.6>]}}}, CurveIndex=$Particles.NormalizedLife)
    Module.Param0WriteEnabled=true
    Module.Param1WriteEnabled=true
    Module.Param2WriteEnabled=true
    Module.Param3WriteEnabled=true
   ) {enabled=false}
   &SubUVAnimation(
    "Sprite Renderer "=DataInterface() {propertyValues {SpriteRenderer=&self:NiagaraEmitter_2.NiagaraSpriteRendererProperties_2}}
    UseEndFrame=false
    UseStartFrame=false
   )
   /SetVariables(
    Particles.0 - Env Fog Dyn Param:Vector4f=&MakeVector4(X=0, Y=1, Z=1, W=1)
    Particles.1 - Ground Fog Dyn Param:Vector4f=&MakeVector4(X=0, Y=1, Z=0.1, W=1)
   ) {isSetParameters=true}
  ]
  rendererClasses [&NiagaraSpriteRendererProperties]
  renderers [{rendererClass=&NiagaraSpriteRendererProperties}]
 }
 burst_around_the_star_sparks001 {
  emitterUpdateScript [
   &EmitterState()
   &SpawnBurst_Instantaneous(Spawn Count=&UniformRangedInt(Minimum=64, Maximum=100), Spawn Time=0.05, Use Spawn Probability=false)
   &SpawnRate(SpawnRate=50, Use Spawn Probability=false) {enabled=false}
  ]
  particleSpawnScript [
   &InitializeParticle(
    Color=<rgba 0.0163 0.8534 1 1>
    Lifetime Max=1.5
    Lifetime Min=0.5
    Lifetime Mode=::NewEnumerator1
    Position Offset=<0 0 1.5>
    Sprite Size Mode=::NewEnumerator1
    Uniform Sprite Size Max=15
    Uniform Sprite Size Min=3
    UsePositionOffset=true
    Write Material Random=true
   )
   &ShapeLocation(
    Disc Coverage=0.7667
    MaskSpawnGroup=false
    Override Seed=false
    Ring / Disc Mode:ENiagaraRingDiscMode=::NewEnumerator1
    Ring Radius=50
    Shape Primitive=::NewEnumerator4
    U Distribution=0.7857
    UseOwnerScale=true
   )
   &AddVelocityFromPoint(
    Module.Use Falloff Distance=false
    Origin Offset=&MakeVector(X=0, Y=0, Z=&RandomRangeFloat(Minimum=-3, Maximum=10))
    Velocity Strength=&Multiply_Float(A=&RandomRangeFloat(Minimum=0.5, Maximum=1), B=500)
   )
   &ApplyInitialForces()
  ]
  particleUpdateScript [
   &ParticleState()
   &ScaleSpriteSize(Scale Factor=&Vector2DFromCurve(
     Vector2Curve=DataInterface() {propertyValues {
       XCurve {keys [<at 0 1 RCIM_Cubic>, {arriveTangent=-0.5084, interpMode=RCIM_Cubic, leaveTangent=-0.5084, tangentMode=RCTM_User, time=1, value=0}]}
       YCurve {keys [{arriveTangent=-3.0155, interpMode=RCIM_Cubic, leaveTangent=-3.0155, tangentMode=RCTM_User, time=0, value=1}, <at 1 0 RCIM_Cubic>]}
      }}
     ScaleCurve=&Vector2DFromFloat(Value=&Lerp_Float(
       A=&FloatFromCurve(
        FloatCurve=DataInterface() {propertyValues {Curve {keys [
            {interpMode=RCIM_Cubic, tangentMode=RCTM_Break, time=0, value=0.5}
            {arriveTangent=0.001, interpMode=RCIM_Cubic, tangentMode=RCTM_Break, time=1000, value=1}
           ]}}}
        CurveIndex=$Engine.Owner.LODDistance
       )
       B=1
       Alpha=1
      ))
    ))
   &CurlNoiseForce(Noise Frequency=5, Noise Strength=500, Pan Noise=true, Pan Noise Field=<0 0.2 0.5>)
   &Drag(Drag=&UniformRangedFloat(Minimum=5, Maximum=15), Use Linear Drag=true, Use Rotational Drag=false)
   &GravityForce(Gravity=&MakeVector(
     X=0
     Y=0
     Z=&FloatFromCurve(
      FloatCurve=DataInterface() {propertyValues {Curve {
         defaultValue=1
         keys [{interpMode=RCIM_Cubic, tangentMode=RCTM_User, time=0, value=0}, {interpMode=RCIM_Cubic, tangentMode=RCTM_User, time=1, value=0.5}]
        }}}
      Scale Curve=980
     )
    ))
   &SolveForcesAndVelocity(Clamp Velocity=false, Limit Acceleration=false)
   &ScaleSpriteSizeBySpeed(
    Max Scale Factor=&RandomRangeVector2D(Minimum=<0.1 1>, Maximum=<0.8 4>)
    Min Scale Factor=<0.1 0.3>
    Velocity Threshold=500
   )
   &ScaleColor(
    Scale Alpha=&FloatFromCurve(
     FloatCurve=DataInterface() {propertyValues {Curve {keys [
         {tangentMode=RCTM_User, time=0, value=1}
         {arriveTangent=-0.0405, interpMode=RCIM_Cubic, leaveTangent=-0.0405, tangentMode=RCTM_User, time=0.7009, value=1}
         {arriveTangent=-8.0089, interpMode=RCIM_Cubic, leaveTangent=-8.0089, tangentMode=RCTM_User, time=1, value=0}
        ]}}}
     Scale Curve=&FNTimeOfDayScalar(Night Scale=10, Morning Scale=7, Day Scale=7, Evening Scale=7)
    )
    Scale RGB=<3 3 3>
    ScaleA=true
    ScaleRGB=true
   )
  ]
  rendererClasses [&NiagaraSpriteRendererProperties]
  renderers [{rendererClass=&NiagaraSpriteRendererProperties}]
 }
}
summary {
 systemName=NS_Emote_ChimeCurlTell_Star_decal
 emitters [
  {emitterName=Spark_GPU001, simTarget=GPUComputeSim, rendererClasses [&NiagaraSpriteRendererProperties]}
  {emitterName=burst_around_the_star_sparks001, rendererClasses [&NiagaraSpriteRendererProperties]}
  {
   emitterName=Glass_Shatter002
   rendererClasses [&NiagaraSpriteRendererProperties, &NiagaraMeshRendererProperties]
  }
  {
   emitterName=Flare_impact_merged
   rendererClasses [&NiagaraSpriteRendererProperties, &NiagaraLightRendererProperties]
  }
  {emitterName=SwirlSmokeHead, rendererClasses [&NiagaraSpriteRendererProperties]}
  {
   emitterName=Star_impact_ground
   rendererClasses [&NiagaraSpriteRendererProperties, &NiagaraMeshRendererProperties]
  }
 ]
 userVariables [
  User.Anim_curve_colorChange {type=Vector3f, description="", defaultValue {struct=Vector3f, value=<1 1 1>}}
  User.Anim_curve_colorSelect {type=NiagaraFloat, defaultValue=0, description=""}
  User.Anim_Curve_spawnRate {type=NiagaraFloat, defaultValue=2, description=""}
  User.bIsFrontEnd {type=NiagaraBool}
  User.bIsFrontEndPreview {type=NiagaraBool}
  User.color curve {type=LinearColor, description="", defaultValue {struct=LinearColor, value=<rgba 1.0781 14.86 15 0.6879>}}
  User.DesertBoulder {type=NiagaraBool}
  User.HitIntensity {type=NiagaraFloat, defaultValue=0, description=""}
  User.OwnerLocalVelocity {type=Vector3f, description="", defaultValue {struct=Vector3f, value=<0 0 0>}}
  User.OwnerSkeleton {
   type=NiagaraDataInterfaceSkeletalMesh
   defaultValue {
    struct=NiagaraExt_VariableValue_DataInterface
    value {dataInterface=&self:NiagaraDataInterfaceSkeletalMesh_1}
   }
  }
 ]
}
