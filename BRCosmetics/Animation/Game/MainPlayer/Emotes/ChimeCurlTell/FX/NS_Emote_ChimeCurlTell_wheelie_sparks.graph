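# Niagara system for the ChimeCurlTell emote's wheelie effect: orange sparks, dust and green exhaust fire with a flickering light.
# Emitters attach to the skeletal mesh, with dust and sparks spawning above an owner-velocity threshold, or at fixed rates in the front end.
# Exhaust and spark emitters condense several layers each, using spawn groups to switch parameters and visibility tags.
# Sparks have GPU and CPU variants, and some brightness values scale with time of day.
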NiagaraSystem /BRCosmetics/Animation/Game/MainPlayer/Emotes/ChimeCurlTell/FX/NS_Emote_ChimeCurlTell_wheelie_sparks
dependencies {
 usedDataInterfaces [
  &NiagaraDataInterfaceCurve
  &NiagaraDataInterfaceArrayFloat
  &NiagaraDataInterfaceVector2DCurve
  &NiagaraDataInterfaceArrayFloat3
  &NiagaraDataInterfaceSpriteRendererInfo
  &NiagaraDataInterfaceVectorCurve
  &NiagaraDataInterfaceColorCurve
  &NiagaraDataInterfaceSkeletalMesh
 ]
 usedDynamicInputs [&FNTimeOfDayColor, &RandomRangeLinearColor]
 usedModules [
  &SystemState
  &self:SystemSpawnScript.SystemScriptSource.SystemScriptGraph.NiagaraNodeAssignment_1.SetVariables_#1
  &self:SystemSpawnScript.SystemScriptSource.SystemScriptGraph.NiagaraNodeAssignment_2.SetVariables_#2
 ]
}
emitters {
 Dust {
  emitterSpawnScript [/SetVariables(Emitter.FE Particle Spawn Rate=60, Emitter.Max Traversal Speed=300, Emitter.Particle Spawn Rate=60) {isSetParameters=true, module=SetVariables_#3}]
  emitterUpdateScript [
   &EmitterState(Life Cycle Mode=::NewEnumerator1, Loop Behavior=::NewEnumerator1, Loop Duration=4, UseLoopDelay=false)
   &SpawnRate(
    SpawnRate=&FN_FrontEndFloatModifier(
     Gameplay=&MakeCustomFloatFromBool(
      Bool=&SetBoolByFloatComparison(A=&VectorLength(Vector=$User.OwnerLocalVelocity), B=100)
      True Float=35
      False Float=0
     )
     Front End=35
    )
    Use Spawn Probability=false
   ) {module=SpawnRate002}
  ]
  particleSpawnScript [
   &PartitionParticles(Number Of Partitions=1)
   &InitializeParticle(
    Color=$System.Global color
    Lifetime=&MakeCustomFloatFromBool(
     Bool=&SetBoolByFloatComparison(A=&VectorLength(Vector=$User.OwnerLocalVelocity), B=150)
     True Float=&RandomRangeFloat(Minimum=0.3, Maximum=0.5)
     False Float=&RandomRangeFloat(Minimum=0.4, Maximum=0.5)
    )
    Position Mode=::NewEnumerator1
    Sprite Rotation Mode=::NewEnumerator2
    Sprite Size Mode=::NewEnumerator1
    Uniform Sprite Size Max=65
    Uniform Sprite Size Min=45
    UsePositionOffset=false
    Write Material Random=true
   ) {module=InitializeParticle001}
   &SkeletalMeshLocation(
    Bone Sampling Mode=::NewEnumerator4
    MaskSpawnGroup=false
    Sampled Position Offset=<0 -12 0>
    Sampled Position Offset Coordinate Space:ENiagaraSkelMeshTransforms=::NewEnumerator2
    Skeletal Mesh=$System.Skeletal Mesh
   )
   &ShapeLocation(
    MaskSpawnGroup=false
    Override Seed=false
    Sphere Radius=&FloatFromCurve(FloatCurve=DataInterface() {propertyValues {Curve {keys [<at 0 0>,<at 1 5>]}}})
    UseOwnerScale=true
   )
   &CameraOffset(Camera Offset Amount=15)
   &AddVelocity(Velocity=<0 -3 0.5>, Velocity Speed Scale=&RandomRangeFloat(Minimum=300, Maximum=500)) {module=AddVelocity002}
  ]
  particleUpdateScript [
   &ParticleState()
   &Drag(
    Drag=&FloatFromCurve(
     FloatCurve=DataInterface() {propertyValues {Curve {
        defaultValue=1
        keys [{interpMode=RCIM_Cubic, tangentMode=RCTM_User, time=0, value=1}, {interpMode=RCIM_Cubic, tangentMode=RCTM_User, time=0.65, value=0.75}]
       }}}
     Scale Curve=&Lerp_Float(
      A=20
      B=20
      Alpha=&RemapRange(
       Input Value To Remap=&VectorLength(Vector=$User.OwnerLocalVelocity)
       Input Min=0
       Input Max=$Emitter.Max Traversal Speed
       Output Min=0
       Output Max=1
      )
     )
    )
    Rotational Drag=5
    Use Linear Drag=true
    Use Rotational Drag=true
   )
   &ScaleSpriteSize(
    Uniform Curve Scale=2
    Uniform Curve Sprite Scale=DataInterface() {propertyValues {Curve {keys [
        {arriveTangent=0.8753, interpMode=RCIM_Cubic, leaveTangent=0.8753, tangentMode=RCTM_User, time=0, value=0.5}
        {arriveTangent=0.0238, interpMode=RCIM_Cubic, leaveTangent=0.0238, tangentMode=RCTM_User, time=0.9843, value=1.7576}
       ]}}}
   )
   &DynamicMaterialParameters(
    Index 0 Param 1=&Add_Float(
     A=&RandomRangeFloat(Minimum=0, Maximum=1)
     B=&FloatFromCurve(
      FloatCurve=DataInterface() {propertyValues {Curve {keys [
          <at 0 0 RCIM_Cubic>
          {arriveTangent=1.5947, interpMode=RCIM_Cubic, leaveTangent=1.5947, tangentMode=RCTM_User, time=0.2557, value=0.5257}
          <at 1 1 RCIM_Cubic>
         ]}}}
      CurveIndex=$Particles.Age
      Scale Curve=0.25
     )
    )
    Module.Index0Float0Write=true
    Module.Index0Float1Write=true
    Module.Index0Float2Write=true
    Module.Index0Float3Write=true
   ) {module=DynamicMaterialParameters001}
   &SolveForcesAndVelocity(Clamp Velocity=false, Limit Acceleration=false)
   &ScaleColor(
    Scale Alpha=&FloatFromCurve(
     FloatCurve=DataInterface() {propertyValues {Curve {keys [
         {arriveTangent=0.0878, interpMode=RCIM_Cubic, leaveTangent=0.0878, tangentMode=RCTM_User, time=0, value=0}
         {
          arriveTangent=0.0878
          arriveTangentWeight=0.0754
          interpMode=RCIM_Cubic
          leaveTangent=0.0878
          leaveTangentWeight=0.2297
          tangentMode=RCTM_User
          time=0.1
          value=1
         }
         {arriveTangent=-0.3893, interpMode=RCIM_Cubic, leaveTangent=-3.5553, tangentMode=RCTM_Break, time=0.4045, value=1}
         {arriveTangent=-0.0368, interpMode=RCIM_Cubic, leaveTangent=-2.9749, tangentMode=RCTM_Break, time=0.9, value=0}
        ]}}}
     Scale Curve=&FN_FrontEndFloatModifier(
      Gameplay=&FNTimeOfDayScalar(Night Scale=0.05, Morning Scale=0.08, Day Scale=0.07, Evening Scale=0.08)
      Front End=0.1
     )
    )
    Scale RGB=&VectorFromFloat(Value=&RandomRangeFloat(Minimum=0.7, Maximum=1))
    ScaleA=true
    ScaleRGB=true
   )
  ]
  rendererClasses [&NiagaraSpriteRendererProperties]
  renderers [
   {rendererClass=&NiagaraSpriteRendererProperties}
   {rendererClass=&NiagaraSpriteRendererProperties, rendererIndex=1}
  ]
 }
 Exhaust_fire_condensed {
  emitterSpawnScript [/SetVariables(Emitter.FE Particle Spawn Rate=60, Emitter.Max Traversal Speed=50, Emitter.Particle Spawn Rate=60) {isSetParameters=true, module=SetVariables_#3}]
  emitterUpdateScript [
   &EmitterState()
   &SpawnBurst_Instantaneous(Spawn Count=2, Spawn Group=6, Use Spawn Probability=false, UseLoopCountLimit=false) {module=SpawnBurst_Instantaneous003}
   &SpawnBurst_Instantaneous(Spawn Group=5, Use Spawn Probability=false, UseLoopCountLimit=false) {module=SpawnBurst_Instantaneous001}
   &SpawnRate(SpawnGroup=4, SpawnRate=25, Use Spawn Probability=false) {module=SpawnRate004}
   &SpawnBurst_Instantaneous(Spawn Count=2, Spawn Group=3, Use Spawn Probability=false, UseLoopCountLimit=false) {module=SpawnBurst_Instantaneous002}
   &SpawnBurst_Instantaneous(Spawn Count=6, Spawn Group=2, Spawn Time=0.15, Use Spawn Probability=false, UseLoopCountLimit=false)
   &SpawnRate(SpawnGroup=1, SpawnRate=20, Use Spawn Probability=false) {module=SpawnRate001}
   &SpawnRate(
    SpawnRate=/Niagara/DynamicInputs/Math/LerpMultipleFloats(
     Float Count:ENiagara_MultipleLerpCount=::NewEnumerator3
     1:NiagaraFloat=&UniformRangedFloat(Minimum=15, Maximum=20)
     2:NiagaraFloat=4
     3:NiagaraFloat=?
     4:NiagaraFloat=?
     5:NiagaraFloat=?
     Alpha:NiagaraFloat=&ClampFloat(Float=&FN_LOD_Distance_Remap(Start Distance=600, End Distance=10000, Minimum Output=0, Maximum Output=1))
    )
    Use Spawn Probability=false
   )
   /SetVariables(Emitter.Flicker:NiagaraFloat=&Add_Float(A=&Multiply_Float(A=$Emitter.Waveform.Waveform, B=0.5), B=0.5)) {isSetParameters=true, module=SetVariables_#4}
  ]
  particleSpawnScript [
   /SetVariables(Particles.VisibilityTag=&SpawnGroup()) {isSetParameters=true, module=SetVariables_#5}
   &PartitionParticles()
   &InitializeParticle(
    Color=&MergeHelper_ColorSwitch(
     0=<rgba 0.0073 1 0 1>
     1=<rgba 0.0073 1 0 1>
     2=<rgba 0.076 1 0.0846 1>
     3=<rgba 1 1 1 0.1>
     4=$System.Global color
     5=<rgba 0.3715 1 0 1>
     Index=&SpawnGroup()
     Num=::NewEnumerator4
    )
    Lifetime=&MergeHelper_FloatSwitch(
     0=&Multiply_Float(A=&UniformRangedFloat(Minimum=0.9, Maximum=1.35), B=0.75)
     1=&RandomRangeFloat(Minimum=0.5, Maximum=0.6)
     2=0.7
     3=4
     4=1
     5=4
     6=1
     Index=&SpawnGroup()
     Num=::NewEnumerator5
    )
    Mesh Scale=&Multiply_VectorByFloat(Vector=<3.5 0.5 1.5>, Float=0.07)
    Mesh Scale Mode=::NewEnumerator4
    Position Mode=::NewEnumerator1
    Sprite Rotation Angle=&MergeHelper_FloatSwitch(
     0=&UniformRangedFloat(Minimum=0, Maximum=180)
     1=&RandomRangeFloat(Minimum=-180, Maximum=200)
     2=&RandomRangeFloat(Minimum=-360, Maximum=360)
     3=100
     4=0
     Index=&SpawnGroup()
     Num=::NewEnumerator3
    )
    Sprite Rotation Mode=::NewEnumerator0
    Sprite Size=&MergeHelper_Vector2Switch(
     0=&Vector2DFromFloat(Value=&Multiply_Float(A=0.4, B=&UniformRangedFloat(Minimum=35, Maximum=50)))
     1=&Vector2DFromFloat(Value=&RandomRangeFloat(Minimum=5, Maximum=10))
     2=&RandomRangeVector2D(Minimum=<85 85>, Maximum=<175 175>)
     3=<5 5>
     4=&RandomRangeVector2D(Minimum=<20 20>, Maximum=<35 35>)
     Index=&SpawnGroup()
     Num=::NewEnumerator3
    )
    Sprite Size Mode=::NewEnumerator4
    UsePositionOffset=false
    Write Material Random=true
   ) {module=InitializeParticle001}
   &ShapeLocation(
    MaskSpawnGroup=false
    Override Seed=false
    Sphere Radius=&FloatFromCurve(FloatCurve=DataInterface() {propertyValues {Curve {keys [<at 0 0>,<at 1 5>]}}})
    UseOwnerScale=true
   )
   &CameraOffset(Camera Offset Amount=&MergeHelper_FloatSwitch(0=5, 1=5, 2=5, 3=-5, 4=5, 5=5, Index=&SpawnGroup(), Num=::NewEnumerator4))
   &AddVelocity(Velocity=&MergeHelper_VectorSwitch(
     0=&RandomRangeVector(Minimum=<0 -50 20>, Maximum=<1 -85 20>)
     1=&RandomRangeVector(Minimum=<0 -75 20>, Maximum=<1 -100 20>)
     2=&MakeVector(
      X=0
      Y=-100
      Z=&FloatFromCurve(FloatCurve=DataInterface() {propertyValues {Curve {keys [<at 0 1>,<at 0.75 0>]}}}, Scale Curve=20)
     )
     3=<0 0 0>
     4=&RandomRangeVector(Minimum=<0 -50 20>, Maximum=<1 -100 20>)
     5=<0 0 0>
     Index=&SpawnGroup()
     Num=::NewEnumerator4
    )) {module=AddVelocity002}
   /SetVariables(
    Particles.0 - Flames Exhaust DynParam 001:Vector4f=&MakeVector4(X=&RandomRangeFloat(Minimum=0, Maximum=4), Y=0, Z=0, W=2.5)
    Particles.1 - body dynparam:Vector4f=&MakeVector4(X=0, Y=&RandomRangeFloat(Minimum=0, Maximum=4), Z=0, W=2.5)
   ) {isSetParameters=true, module=SetVariables_#6}
   &InitialMeshOrientation(Mesh Axis To Orient=<0 0 1>, Mesh Orientation Mode=::NewEnumerator2, Orientation Vector=<0 1 0.6>, Use Rotation Vector=false)
   /SetVariables(Particles.Fire filler Orientation:Quat4f=<-0.2606 0.2313 1 0>) {isSetParameters=true}
  ]
  particleUpdateScript [
   &ParticleState()
   &SubUVAnimation(
    Random Start Frame=true
    "Sprite Renderer "=DataInterface() {propertyValues {SpriteRenderer=&self:Exhaust_fire_condensed_0.NiagaraSpriteRendererProperties_5}}
    UseEndFrame=false
    UseStartFrame=false
   )
   &ScaleSpriteSize(
    Scale Factor=&MergeHelper_Vector2Switch(
     0=&Vector2DFromCurve(
      Vector2Curve=DataInterface() {propertyValues {
        XCurve {keys [<at 0 0.3 RCIM_Cubic>, {arriveTangent=0.3, interpMode=RCIM_Cubic, leaveTangent=0.3, time=0.3, value=1}, <at 1 0.6>]}
        YCurve {keys [<at 0 0.3 RCIM_Cubic>, {arriveTangent=0.3, interpMode=RCIM_Cubic, leaveTangent=0.3, time=0.3, value=1}, <at 1 0.6>]}
       }}
      ScaleCurve=<1 1>
     )
     1=&Multiply_Vector2D_ByFloat(
      Float=&FloatFromCurve(FloatCurve=DataInterface() {propertyValues {Curve {keys [
           {arriveTangent=1.4708, interpMode=RCIM_Cubic, leaveTangent=1.4708, tangentMode=RCTM_User, time=0, value=0.3}
           {arriveTangent=1.7, interpMode=RCIM_Cubic, leaveTangent=1.7, time=0.6, value=1.625}
           {interpMode=RCIM_Cubic, tangentMode=RCTM_User, time=1, value=2}
          ]}}})
      Vector2D=<1 1>
     )
     2=&Multiply_Vector2D_ByFloat(
      Float=&FloatFromCurve(
       FloatCurve=DataInterface() {propertyValues {Curve {keys [
           {arriveTangent=1.4708, interpMode=RCIM_Cubic, leaveTangent=1.4708, tangentMode=RCTM_User, time=0, value=0}
           {arriveTangent=-0.0652, interpMode=RCIM_Cubic, leaveTangent=-0.0652, tangentMode=RCTM_User, time=0.2977, value=1.7547}
           {interpMode=RCIM_Cubic, tangentMode=RCTM_User, time=1, value=2}
          ]}}}
       Scale Curve=0.15
      )
      Vector2D=<1 1>
     )
     3=<1 1>
     4=&Vector2DFromFloat(Value=&FloatFromCurve(FloatCurve=DataInterface() {propertyValues {Curve {keys [
           {arriveTangent=0.8753, interpMode=RCIM_Cubic, leaveTangent=0.8753, tangentMode=RCTM_User, time=0, value=0}
           {arriveTangent=0.0238, interpMode=RCIM_Cubic, leaveTangent=0.0238, tangentMode=RCTM_User, time=1, value=1.7}
          ]}}}))
     Index=&SpawnGroup()
     Num=::NewEnumerator3
    )
    Scale Sprite Size Mode=::NewEnumerator2
   )
   &SkeletalMeshLocation(
    Bone / Socket Index=$Particles.PartitionParticles.Partition
    MaskSpawnGroup=false
    Mesh Sampling Type=::NewEnumerator5
    Position Sampling=::NewEnumerator1
    Sampled Position Offset=&MergeHelper_VectorSwitch(
     0=<0 3 -3>
     1=<0 7 -3>
     2=<0 7 -3>
     3=<0 3 0>
     4=<0 7 -3>
     5=<0 7 -3>
     Index=&SpawnGroup()
     Num=::NewEnumerator4
    )
    Sampled Position Offset Coordinate Space:ENiagaraSkelMeshTransforms=::NewEnumerator2
    Skeletal Mesh=$System.Skeletal Mesh
    Socket Sampling Mode=::NewEnumerator5
   )
   &ScaleMeshSize(Scale Factor=&MakeVector(
     X=0.5
     Y=&FloatFromCurve(
      FloatCurve=DataInterface() {propertyValues {Curve {
         defaultValue=1
         keys [{interpMode=RCIM_Cubic, tangentMode=RCTM_User, time=0, value=0}, {interpMode=RCIM_Cubic, tangentMode=RCTM_User, time=0.05, value=1}]
        }}}
      Scale Curve=0.2
     )
     Z=0.5
    ))
   &Drag(
    Drag=&MergeHelper_FloatSwitch(
     0=3
     1=&FloatFromCurve(FloatCurve=DataInterface() {propertyValues {Curve {keys [<at 0 0.075 RCIM_Cubic>,<at 1 1.075 RCIM_Cubic>]}}}, Scale Curve=10)
     2=0
     3=0
     4=0
     Index=&SpawnGroup()
     Num=::NewEnumerator3
    )
    Rotational Drag=5
    Use Linear Drag=true
    Use Rotational Drag=true
   )
   &AccelerationForce(
    Acceleration=&MergeHelper_VectorSwitch(
     0=&UniformRangedVector(Minimum=<-20 -147 30>, Maximum=<20 -205 30>)
     1=<0 0 0>
     2=<0 0 0>
     3=<0 0 0>
     4=<0 0 0>
     Index=&SpawnGroup()
     Num=::NewEnumerator3
    )
    Coordinate Space=::Local
   )
   &SolveForcesAndVelocity(Clamp Velocity=false, Limit Acceleration=false)
   &ScaleSpriteSizeBySpeed(
    Sample Scale Factor By Curve=true
    Scale Factor Curve=DataInterface() {propertyValues {Curve {keys [<at 0 0>,<at 1 2>]}}}
    Velocity Threshold=100
   )
   &ScaleColor(
    Scale Alpha=&MergeHelper_FloatSwitch(
     0=&Multiply_Float(
      A=0.5
      B=&FloatFromCurve(FloatCurve=DataInterface() {propertyValues {Curve {keys [
           {arriveTangent=2.4553, interpMode=RCIM_Cubic, leaveTangent=2.4553, time=0, value=0}
           {arriveTangent=-0.0116, interpMode=RCIM_Cubic, leaveTangent=-0.0116, tangentMode=RCTM_User, time=0.4073, value=1}
           {arriveTangent=-2.0743, interpMode=RCIM_Cubic, leaveTangent=-2.0743, tangentMode=RCTM_User, time=0.7, value=0.2}
           {arriveTangent=-0.6667, interpMode=RCIM_Cubic, leaveTangent=-0.6667, time=1, value=0}
          ]}}})
     )
     1=&FloatFromCurve(
      FloatCurve=DataInterface() {propertyValues {Curve {keys [
          {arriveTangent=6.7724, interpMode=RCIM_Cubic, leaveTangent=6.7724, tangentMode=RCTM_User, time=0, value=0}
          {interpMode=RCIM_Cubic, tangentMode=RCTM_User, time=0.3949, value=1}
          {arriveTangent=-1.6527, interpMode=RCIM_Cubic, leaveTangent=-1.6527, time=0.7199, value=0.2109}
          {arriveTangent=-0.174, interpMode=RCIM_Cubic, leaveTangent=-0.174, tangentMode=RCTM_User, time=1, value=0}
         ]}}}
      Scale Curve=&FNTimeOfDayScalar(Night Scale=1.5, Morning Scale=0.5, Day Scale=0.52, Evening Scale=0.5)
     )
     2=&FloatFromCurve(
      FloatCurve=DataInterface() {propertyValues {Curve {keys [
          {arriveTangent=6.7724, interpMode=RCIM_Cubic, leaveTangent=6.7724, tangentMode=RCTM_User, time=0, value=0}
          {interpMode=RCIM_Cubic, tangentMode=RCTM_User, time=0.25, value=1}
          {arriveTangent=-0.174, interpMode=RCIM_Cubic, leaveTangent=-0.174, tangentMode=RCTM_User, time=0.6, value=0}
         ]}}}
      Scale Curve=&FNTimeOfDayScalar(Night Scale=1.5, Morning Scale=0.5, Day Scale=0.52, Evening Scale=0.5)
     )
     3=1
     4=&FloatFromCurve(
      FloatCurve=DataInterface() {propertyValues {Curve {keys [
          {arriveTangent=0.0878, interpMode=RCIM_Cubic, leaveTangent=0.0878, tangentMode=RCTM_User, time=0, value=0}
          {
           arriveTangent=0.0878
           arriveTangentWeight=0.0754
           interpMode=RCIM_Cubic
           leaveTangent=0.0878
           leaveTangentWeight=0.2297
           tangentMode=RCTM_User
           time=0.1684
           value=1
          }
          {arriveTangent=-0.3893, interpMode=RCIM_Cubic, leaveTangent=-3.5553, tangentMode=RCTM_Break, time=0.4045, value=1}
          {arriveTangent=-0.0368, interpMode=RCIM_Cubic, leaveTangent=-2.9749, tangentMode=RCTM_Break, time=0.9, value=0}
         ]}}}
      Scale Curve=&FN_FrontEndFloatModifier(
       Gameplay=&FNTimeOfDayScalar(Night Scale=0.15, Morning Scale=0.1, Day Scale=0.1, Evening Scale=0.1)
       Front End=0.4
      )
     )
     Index=&SpawnGroup()
     Num=::NewEnumerator3
    )
    Scale RGB=&MergeHelper_VectorSwitch(
     0=&VectorFromCurve(VectorCurve=DataInterface() {propertyValues {
        XCurve {keys [<at 0.3242 3 RCIM_Cubic>,<at 0.9878 0.9731>]}
        YCurve {keys [<at 0.3242 2.8 RCIM_Cubic>,<at 1 0.9731>]}
        ZCurve {keys [<at 0.3169 2.6605 RCIM_Cubic>,<at 1 0.9731>]}
       }})
     1=<1 1 1>
     2=&MakeVectorFromLinearColorRGB(Color=&ColorFromCurve(ColorCurve=DataInterface() {propertyValues {
         AlphaCurve {keys [{arriveTangent=-0.1596, interpMode=RCIM_Cubic, leaveTangent=-0.1596, tangentMode=RCTM_User, time=0.8987, value=1}, <at 0.982 0 RCIM_Cubic>]}
         BlueCurve {keys [<at 0.0805 0.6018>,<at 0.9823 0.873>]}
         GreenCurve {keys [<at 0.0805 1>,<at 0.9823 1>]}
         RedCurve {keys [<at 0.0805 0.4511>,<at 0.9823 0.0611>]}
        }}))
     3=<0 1 0>
     4=&VectorFromFloat(Value=&RandomRangeFloat(Minimum=0.3, Maximum=0.5))
     Index=&SpawnGroup()
     Num=::NewEnumerator3
    )
    ScaleA=true
    ScaleRGB=true
   ) {module=ScaleColor001}
   /SetVariables(
    Particles.4 - smoke DynParam:Vector4f=&MakeVector4(
     X=&Add_Float(
      A=&RandomRangeFloat(Minimum=0, Maximum=1)
      B=&FloatFromCurve(
       FloatCurve=DataInterface() {propertyValues {Curve {keys [
           <at 0 0 RCIM_Cubic>
           {arriveTangent=1.5947, interpMode=RCIM_Cubic, leaveTangent=1.5947, tangentMode=RCTM_User, time=0.2557, value=0.5257}
           <at 1 1 RCIM_Cubic>
          ]}}}
       CurveIndex=$Particles.Age
       Scale Curve=0.25
      )
     )
     Y=0
     Z=0
     W=0
    )
    Particles.Flames Exhaust DynParam:Vector4f=&MakeVector4(
     X=&FloatFromCurve(FloatCurve=DataInterface() {propertyValues {Curve {keys [<at 0 1>]}}})
     Y=0.5
     Z=&FloatFromCurve(FloatCurve=DataInterface() {propertyValues {Curve {keys [<at 0 0.5 RCIM_Cubic>, {arriveTangent=2.0035, interpMode=RCIM_Cubic, leaveTangent=2.0035, time=0.2516, value=0.4588}, <at 0.7282 1.9588>]}}})
     W=2
    )
   ) {isSetParameters=true, module=SetVariables_#7}
   &SpriteRotationRate(Rotation Rate=&Lerp_Float(A=-10, B=10, Alpha=$Particles.Initial.MaterialRandom)) {module=SpriteRotationRate001}
   &FN_Light_Attributes(
    Color=<rgba 0.609 5 1.1003 1>
    Create Color Attribute=true
    Create Exponent Attribute=true
    Create Position Attribute=true
    Create Radius Attribute=true
    Enable Particle Light Enable=true
    Enable Visibility Tag=true
    Exponent:NiagaraFloat=&FN_FrontEndFloatModifier(Gameplay=5, Front End=8)
    Light Enabled=true
    Position=&SelectVectorFromArray(
     Direct Array Index=$Particles.PartitionParticles.Partition
     Vector Selection Array=DataInterface() {propertyValues {FloatData [<0 -68 10>, <16 -55 10>]}}
    )
    Radius=&FN_FrontEndFloatModifier(Gameplay=50, Front End=20)
    Scale Color=&Multiply_Float(
     A=0.5
     B=&Multiply_Float(
      A=&Waveform("Waveform [1]"=::NewEnumerator0, "[1] Frequency"=3, "[1] Amplitude Min/Max"=<0.7 1>, "[1] Scale"=2.6368)
      B=&FNTimeOfDayScalar(Night Scale=0.5, Morning Scale=0.6, Day Scale=0.45, Evening Scale=0.6)
     )
    )
    VisibilityTag=5
   )
   &SubUVAnimation(
    Random Start Frame=true
    "Sprite Renderer "=DataInterface() {propertyValues {SpriteRenderer=&self:Exhaust_fire_condensed_0.NiagaraSpriteRendererProperties_3}}
    UseEndFrame=false
    UseStartFrame=false
   ) {module=SubUVAnimation001}
   /SetVariables(
    Particles.3 - Hot Core Glow:LinearColor=&MakeLinearColorFromVectorAndFloat(
     "Vector (RGB)"=&Lerp_Vector(
      A=&MakeVectorFromLinearColorRGB(Color=<rgba 0.3042 4 0.3383 1>)
      B=&MakeVectorFromLinearColorRGB(Color=<rgba 0.3802 5 0.4228 1>)
      Alpha=1
     )
     "Float (Alpha)"=&FloatFromCurve(FloatCurve=DataInterface() {propertyValues {Curve {keys [<at 0 0.5>,<at 0.9944 0>]}}})
    )
    Particles.3 - Hot Core Scale:Vector3f=<0.12 0.12 0.2>
    Particles.3 -fire filler scale:Vector3f=<0.3 0.5 0.5>
   ) {isSetParameters=true}
  ]
  rendererClasses [
   &NiagaraSpriteRendererProperties
   &NiagaraMeshRendererProperties
   &NiagaraLightRendererProperties
  ]
  renderers [
   {rendererClass=&NiagaraSpriteRendererProperties}
   {rendererClass=&NiagaraSpriteRendererProperties, rendererIndex=1}
   {rendererClass=&NiagaraSpriteRendererProperties, rendererIndex=2}
   {rendererClass=&NiagaraMeshRendererProperties, rendererIndex=3}
   {rendererClass=&NiagaraMeshRendererProperties, rendererIndex=4}
   {rendererClass=&NiagaraSpriteRendererProperties, rendererIndex=5}
   {rendererClass=&NiagaraLightRendererProperties, rendererIndex=6}
  ]
 }
 Sparks_condensed {
  simTarget=GPUComputeSim
  emitterUpdateScript [
   &EmitterState()
   &SpawnRate(
    SpawnGroup=2
    SpawnRate=&FN_FrontEndFloatModifier(
     Gameplay=&MakeCustomFloatFromBool(
      Bool=&SetBoolByFloatComparison(A=&VectorLength(Vector=$User.OwnerLocalVelocity), B=150)
      True Float=200
      False Float=0
     )
     Front End=250
    )
    Use Spawn Probability=false
   ) {module=SpawnRate001}
   &SpawnRate(
    SpawnGroup=1
    SpawnRate=&FN_FrontEndFloatModifier(
     Gameplay=&MakeCustomFloatFromBool(
      Bool=&SetBoolByFloatComparison(A=&VectorLength(Vector=$User.OwnerLocalVelocity), B=150)
      True Float=400
      False Float=0
     )
     Front End=250
    )
    Use Spawn Probability=false
   )
   &SpawnRate(
    SpawnRate=&FN_FrontEndFloatModifier(
     Gameplay=&MakeCustomFloatFromBool(
      Bool=&SetBoolByFloatComparison(A=&VectorLength(Vector=$User.OwnerLocalVelocity), B=150)
      True Float=400
      False Float=0
     )
     Front End=250
    )
    Use Spawn Probability=false
   ) {module=SpawnRate002}
  ]
  particleSpawnScript [
   /SetVariables(Particles.VisibilityTag=&SpawnGroup()) {isSetParameters=true, module=SetVariables_#8}
   &InitializeParticle(
    Color=&MergeHelper_ColorSwitch(
     0=<rgba 4.3698 2.6465 0.4346 1>
     1=<rgba 4.3698 2.6465 0.4346 1>
     2=<rgba 4.3698 1.9199 0.4158 1>
     3=<rgba 4.3698 2.6465 0.4346 1>
     4=<rgba 0 0 0 0>
     Index=&SpawnGroup()
     Num=::NewEnumerator3
    )
    Lifetime=&MergeHelper_FloatSwitch(
     0=&RandomRangeFloat(Minimum=0.05, Maximum=0.2)
     1=&RandomRangeFloat(Minimum=0.1, Maximum=0.2)
     2=0.07
     3=0.3
     4=0
     Index=&SpawnGroup()
     Num=::NewEnumerator3
    )
    Sprite Size=&MergeHelper_Vector2Switch(
     0=&RandomRangeVector2D(Minimum=<1 10>, Maximum=<1.5 15>)
     1=&RandomRangeVector2D(Minimum=<0.5 0.5>, Maximum=<2.5 2.5>)
     2=&RandomRangeVector2D(Minimum=<0.2 5>, Maximum=<1 8>)
     3=<0 0>
     4=<0 0>
     Index=&SpawnGroup()
     Num=::NewEnumerator3
    )
    Sprite Size Mode=::NewEnumerator4
    UsePositionOffset=false
    Write Material Random=true
   )
   &SkeletalMeshLocation(
    Bone Sampling Mode=::NewEnumerator4
    MaskSpawnGroup=false
    Sampled Position Offset=&MergeHelper_VectorSwitch(0=<0 -45 3>, 1=<0 -50 0>, 2=<0 -60 0>, 3=<0 0 0>, 4=<0 0 0>, Index=&SpawnGroup(), Num=::NewEnumerator3)
    Sampled Position Offset Coordinate Space:ENiagaraSkelMeshTransforms=::NewEnumerator2
    Skeletal Mesh=$System.Skeletal Mesh
   )
   &AddVelocity(
    Cone Angle=&MergeHelper_FloatSwitch(0=40, 1=40, 2=60, 3=0, 4=0, Index=&SpawnGroup(), Num=::NewEnumerator3)
    Cone Axis=&MergeHelper_VectorSwitch(
     0=<0 -1 0.3844>
     1=<0 -1 0.3844>
     2=<0 -1 0.3844>
     3=<0 0 0>
     4=<0 0 0>
     Index=&SpawnGroup()
     Num=::NewEnumerator3
    )
    Module.UseConeFalloff=false
    Module.UseVelDistribution=false
    Override Seed=false
    Velocity Mode=::NewEnumerator2
    Velocity Speed=&MergeHelper_FloatSwitch(
     0=1000
     1=&RandomRangeFloat(Minimum=500, Maximum=1200)
     2=&RandomRangeFloat(Minimum=500, Maximum=1000)
     3=0
     4=0
     Index=/Niagara/DynamicInputs/Math/GetFloatArrayCount(Float Array:NiagaraDataInterfaceArrayFloat=DataInterface() {propertyValues {FloatData []}})
     Num=::NewEnumerator3
    )
   )
  ]
  particleUpdateScript [
   &ParticleState()
   &ScaleSpriteSize(
    Scale Factor=&MergeHelper_Vector2Switch(
     0=&MakeVector2D(
      X=0.5
      Y=&FloatFromCurve(
       FloatCurve=DataInterface() {propertyValues {Curve {keys [<at -0.0046 0>,<at 0.0873 0.9701>,<at 1.0011 0.002>]}}}
       CurveIndex=&RandomRangeFloat(Minimum=0.5, Maximum=1)
      )
     )
     1=&Vector2DFromFloat(Value=&Multiply_Float(
       A=&FloatFromCurve(
        FloatCurve=DataInterface() {propertyValues {Curve {keys [
            {interpMode=RCIM_Cubic, tangentMode=RCTM_Break, time=0, value=0}
            {interpMode=RCIM_Cubic, tangentMode=RCTM_Break, time=0.1, value=1}
            {interpMode=RCIM_Cubic, tangentMode=RCTM_Break, time=0.2, value=0}
            {arriveTangent=5.8511, interpMode=RCIM_Cubic, leaveTangent=-6.2653, tangentMode=RCTM_Break, time=0.3023, value=0.9569}
            {interpMode=RCIM_Cubic, tangentMode=RCTM_Break, time=0.4029, value=0}
            {arriveTangent=7.3852, interpMode=RCIM_Cubic, leaveTangent=-4.8986, tangentMode=RCTM_Break, time=0.5198, value=0.7267}
            {interpMode=RCIM_Cubic, tangentMode=RCTM_Break, time=0.6539, value=0}
            {interpMode=RCIM_Cubic, leaveTangent=-5.1375, tangentMode=RCTM_Break, time=0.7904, value=0.4154}
            {interpMode=RCIM_Cubic, tangentMode=RCTM_Break, time=0.9977, value=0}
           ]}}}
        Scale Curve=&RandomRangeFloat(Minimum=0.2, Maximum=0.35)
       )
       B=&Lerp_Float(
        A=5
        B=&FloatFromCurve(
         FloatCurve=DataInterface() {propertyValues {Curve {keys [
             {interpMode=RCIM_Cubic, tangentMode=RCTM_Break, time=0, value=0}
             {interpMode=RCIM_Cubic, tangentMode=RCTM_Break, time=0.1, value=1}
             {interpMode=RCIM_Cubic, tangentMode=RCTM_Break, time=0.2, value=0}
             {arriveTangent=5.8511, interpMode=RCIM_Cubic, leaveTangent=-6.2653, tangentMode=RCTM_Break, time=0.3023, value=0.9569}
             {interpMode=RCIM_Cubic, tangentMode=RCTM_Break, time=0.4029, value=0}
             {arriveTangent=7.3852, interpMode=RCIM_Cubic, leaveTangent=-4.8986, tangentMode=RCTM_Break, time=0.5198, value=0.7267}
             {interpMode=RCIM_Cubic, tangentMode=RCTM_Break, time=0.6539, value=0}
             {interpMode=RCIM_Cubic, leaveTangent=-5.1375, tangentMode=RCTM_Break, time=0.7904, value=0.4154}
             {interpMode=RCIM_Cubic, tangentMode=RCTM_Break, time=0.9977, value=0}
            ]}}}
         CurveIndex=&Multiply_Float(A=$Particles.NormalizedAge, B=&RandomRangeFloat(Minimum=0.1, Maximum=0.5))
         Scale Curve=7
        )
        Alpha=0.7
       )
      ))
     2=&MakeVector2D(
      X=0.5
      Y=&FloatFromCurve(
       FloatCurve=DataInterface() {propertyValues {Curve {keys [<at -0.0046 0>,<at 0.0873 0.9701>,<at 1.0011 0.002>]}}}
       CurveIndex=&RandomRangeFloat(Minimum=0.5, Maximum=1)
      )
     )
     3=<1 1>
     4=<1 1>
     Index=&SpawnGroup()
     Num=::NewEnumerator3
    )
    Scale Sprite Size Mode=::NewEnumerator2
   )
   &GravityForce(Gravity=&MergeHelper_VectorSwitch(
     0=&RandomRangeVector(Minimum=<0 0 200>, Maximum=<0 0 -980>)
     1=&RandomRangeVector(Minimum=<0 0 350>, Maximum=<0 0 -1500>)
     2=&RandomRangeVector(Minimum=<0 0 200>, Maximum=<1 1 -980>)
     3=<0 0 0>
     4=<0 0 0>
     Index=&SpawnGroup()
     Num=::NewEnumerator3
    ))
   &CurlNoiseForce(
    Bias Noise=false
    Noise Frequency=&MergeHelper_FloatSwitch(0=50, 1=50, 2=0, 3=0, 4=0, Index=&SpawnGroup(), Num=::NewEnumerator3)
    Noise Strength=&MergeHelper_FloatSwitch(0=50, 1=50, 2=0, 3=0, 4=0, Index=&SpawnGroup(), Num=::NewEnumerator3)
    Pan Noise=false
   )
   &Drag(
    Drag=&MergeHelper_FloatSwitch(
     0=&RandomRangeFloat(Minimum=5, Maximum=30)
     1=&RandomRangeFloat(Minimum=5, Maximum=20)
     2=0
     3=0
     4=0
     Index=&SpawnGroup()
     Num=::NewEnumerator3
    )
    Rotational Drag:NiagaraFloat=&MergeHelper_FloatSwitch(0=0, 1=0, 2=0, 3=0, 4=0, Index=&SpawnGroup(), Num=::NewEnumerator3)
    Use Linear Drag=true
    Use Rotational Drag=true
   )
   &SolveForcesAndVelocity(Clamp Velocity=false, Limit Acceleration=false)
   &ScaleSpriteSizeBySpeed(
    Max Scale Factor=&MergeHelper_Vector2Switch(0=<1.2 3>, 1=<2 2>, 2=<1.2 1>, 3=<0 0>, 4=<0 0>, Index=&SpawnGroup(), Num=::NewEnumerator3)
    Min Scale Factor=&MergeHelper_Vector2Switch(0=<0.5 1.5>, 1=<1 1>, 2=<1 1.5>, 3=<0 0>, 4=<0 0>, Index=&SpawnGroup(), Num=::NewEnumerator3)
    Velocity Threshold:NiagaraFloat=&MergeHelper_FloatSwitch(0=250, 1=250, 2=250, 3=0, 4=0, Index=&SpawnGroup(), Num=::NewEnumerator3)
   )
   &ScaleColor(
    Scale Alpha=&MergeHelper_FloatSwitch(
     0=&FloatFromCurve(FloatCurve=DataInterface() {propertyValues {Curve {keys [<at -0.0092 0.0144>,<at 0.0967 0.9767>,<at 1.0011 0.002>]}}})
     1=&FloatFromCurve(FloatCurve=DataInterface() {propertyValues {Curve {keys [<at -0.0092 0.0144>,<at 0.0967 0.9767>,<at 1.0011 0.002>]}}})
     2=&FloatFromCurve(FloatCurve=DataInterface() {propertyValues {Curve {keys [<at -0.0092 0.0144>,<at 0.0967 0.9767>,<at 1.0011 0.002>]}}})
     3=1
     4=1
     Index=&SpawnGroup()
     Num=::NewEnumerator3
    )
    Scale RGB=&MergeHelper_VectorSwitch(0=<15 5 1>, 1=<15 5 1>, 2=<15 5 1>, 3=<1 1 1>, 4=<1 1 1>, Index=&SpawnGroup(), Num=::NewEnumerator3)
    ScaleA=true
    ScaleRGB=true
   )
   /SetVariables(
    Particles.Sparks Color:LinearColor=&MakeLinearColorFromVectorAndFloat("Vector (RGB)"=<15 5 1>, "Float (Alpha)"=1)
    Particles.Sparks SpriteSize:Vector2f=&Multiply_Vector2D_ByFloat(Float=1, Vector2D=&RandomRangeVector2D(Minimum=<0.05 0.5>, Maximum=<0.1 2.5>))
   ) {isSetParameters=true, module=SetVariables_#9}
  ]
  rendererClasses [&NiagaraSpriteRendererProperties]
  renderers [
   {rendererClass=&NiagaraSpriteRendererProperties}
   {rendererClass=&NiagaraSpriteRendererProperties, rendererIndex=1}
   {rendererClass=&NiagaraSpriteRendererProperties, rendererIndex=2}
  ]
 }
 Sparks_condensed001 {
  emitterUpdateScript [
   &EmitterState()
   &SpawnRate(
    SpawnGroup=2
    SpawnRate=&FN_FrontEndFloatModifier(
     Gameplay=&MakeCustomFloatFromBool(
      Bool=&SetBoolByFloatComparison(A=&VectorLength(Vector=$User.OwnerLocalVelocity), B=150)
      True Float=250
      False Float=250
     )
     Front End=250
    )
    Use Spawn Probability=false
   ) {enabled=false, module=SpawnRate001}
   &SpawnRate(
    SpawnGroup=1
    SpawnRate=&FN_FrontEndFloatModifier(
     Gameplay=&MakeCustomFloatFromBool(
      Bool=&SetBoolByFloatComparison(A=&VectorLength(Vector=$User.OwnerLocalVelocity), B=150)
      True Float=250
      False Float=0
     )
     Front End=250
    )
    Use Spawn Probability=false
   )
   &SpawnRate(
    SpawnRate=&FN_FrontEndFloatModifier(
     Gameplay=&MakeCustomFloatFromBool(
      Bool=&SetBoolByFloatComparison(A=&VectorLength(Vector=$User.OwnerLocalVelocity), B=150)
      True Float=250
      False Float=0
     )
     Front End=250
    )
    Use Spawn Probability=false
   ) {module=SpawnRate002}
  ]
  particleSpawnScript [
   /SetVariables(Particles.VisibilityTag=&SpawnGroup()) {isSetParameters=true, module=SetVariables_#8}
   &InitializeParticle(
    Color=&MergeHelper_ColorSwitch(
     0=<rgba 4.3698 2.6465 0.4346 1>
     1=<rgba 4.3698 2.6465 0.4346 1>
     2=<rgba 4.3698 1.9199 0.4158 1>
     3=<rgba 4.3698 2.6465 0.4346 1>
     4=<rgba 0 0 0 0>
     Index=&SpawnGroup()
     Num=::NewEnumerator3
    )
    Lifetime=&MergeHelper_FloatSwitch(
     0=&RandomRangeFloat(Minimum=0.05, Maximum=0.2)
     1=&RandomRangeFloat(Minimum=0.1, Maximum=0.2)
     2=0.07
     3=0.3
     4=0
     Index=&SpawnGroup()
     Num=::NewEnumerator3
    )
    Sprite Size=&MergeHelper_Vector2Switch(
     0=&RandomRangeVector2D(Minimum=<2 10>, Maximum=<2.5 15>)
     1=&RandomRangeVector2D(Minimum=<1.5 1.5>, Maximum=<2.5 2.5>)
     2=&RandomRangeVector2D(Minimum=<0.2 5>, Maximum=<1 8>)
     3=<0 0>
     4=<0 0>
     Index=&SpawnGroup()
     Num=::NewEnumerator3
    )
    Sprite Size Mode=::NewEnumerator4
    UsePositionOffset=false
    Write Material Random=true
   )
   &SkeletalMeshLocation(
    Bone Sampling Mode=::NewEnumerator4
    MaskSpawnGroup=false
    Sampled Position Offset=&MergeHelper_VectorSwitch(0=<0 -45 3>, 1=<0 -50 0>, 2=<0 -60 0>, 3=<0 0 0>, 4=<0 0 0>, Index=&SpawnGroup(), Num=::NewEnumerator3)
    Sampled Position Offset Coordinate Space:ENiagaraSkelMeshTransforms=::NewEnumerator2
    Skeletal Mesh=$System.Skeletal Mesh
   )
   &AddVelocity(
    Cone Angle=&MergeHelper_FloatSwitch(0=40, 1=40, 2=60, 3=0, 4=0, Index=&SpawnGroup(), Num=::NewEnumerator3)
    Cone Axis=&MergeHelper_VectorSwitch(
     0=<0 -1 0.3844>
     1=<0 -1 0.3844>
     2=<0 -1 0.3844>
     3=<0 0 0>
     4=<0 0 0>
     Index=&SpawnGroup()
     Num=::NewEnumerator3
    )
    Module.UseConeFalloff=false
    Module.UseVelDistribution=false
    Override Seed=false
    Velocity Mode=::NewEnumerator2
    Velocity Speed=&MergeHelper_FloatSwitch(
     0=1000
     1=&RandomRangeFloat(Minimum=500, Maximum=1200)
     2=&RandomRangeFloat(Minimum=500, Maximum=1000)
     3=0
     4=0
     Index=&GetFloatArrayCount(Float Array:NiagaraDataInterfaceArrayFloat=DataInterface() {propertyValues {FloatData []}})
     Num=::NewEnumerator3
    )
   )
  ]
  particleUpdateScript [
   &ParticleState()
   &ScaleSpriteSize(
    Scale Factor=&MergeHelper_Vector2Switch(
     0=&MakeVector2D(
      X=0.5
      Y=&FloatFromCurve(
       FloatCurve=DataInterface() {propertyValues {Curve {keys [<at -0.0046 0>,<at 0.0873 0.9701>,<at 1.0011 0.002>]}}}
       CurveIndex=&RandomRangeFloat(Minimum=0.5, Maximum=1)
      )
     )
     1=&Vector2DFromFloat(Value=&Multiply_Float(
       A=&FloatFromCurve(
        FloatCurve=DataInterface() {propertyValues {Curve {keys [
            {interpMode=RCIM_Cubic, tangentMode=RCTM_Break, time=0, value=0}
            {interpMode=RCIM_Cubic, tangentMode=RCTM_Break, time=0.1, value=1}
            {interpMode=RCIM_Cubic, tangentMode=RCTM_Break, time=0.2, value=0}
            {arriveTangent=5.8511, interpMode=RCIM_Cubic, leaveTangent=-6.2653, tangentMode=RCTM_Break, time=0.3023, value=0.9569}
            {interpMode=RCIM_Cubic, tangentMode=RCTM_Break, time=0.4029, value=0}
            {arriveTangent=7.3852, interpMode=RCIM_Cubic, leaveTangent=-4.8986, tangentMode=RCTM_Break, time=0.5198, value=0.7267}
            {interpMode=RCIM_Cubic, tangentMode=RCTM_Break, time=0.6539, value=0}
            {interpMode=RCIM_Cubic, leaveTangent=-5.1375, tangentMode=RCTM_Break, time=0.7904, value=0.4154}
            {interpMode=RCIM_Cubic, tangentMode=RCTM_Break, time=0.9977, value=0}
           ]}}}
        Scale Curve=&RandomRangeFloat(Minimum=0.35, Maximum=0.5)
       )
       B=&Lerp_Float(
        A=5
        B=&FloatFromCurve(
         FloatCurve=DataInterface() {propertyValues {Curve {keys [
             {interpMode=RCIM_Cubic, tangentMode=RCTM_Break, time=0, value=0}
             {interpMode=RCIM_Cubic, tangentMode=RCTM_Break, time=0.1, value=1}
             {interpMode=RCIM_Cubic, tangentMode=RCTM_Break, time=0.2, value=0}
             {arriveTangent=5.8511, interpMode=RCIM_Cubic, leaveTangent=-6.2653, tangentMode=RCTM_Break, time=0.3023, value=0.9569}
             {interpMode=RCIM_Cubic, tangentMode=RCTM_Break, time=0.4029, value=0}
             {arriveTangent=7.3852, interpMode=RCIM_Cubic, leaveTangent=-4.8986, tangentMode=RCTM_Break, time=0.5198, value=0.7267}
             {interpMode=RCIM_Cubic, tangentMode=RCTM_Break, time=0.6539, value=0}
             {interpMode=RCIM_Cubic, leaveTangent=-5.1375, tangentMode=RCTM_Break, time=0.7904, value=0.4154}
             {interpMode=RCIM_Cubic, tangentMode=RCTM_Break, time=0.9977, value=0}
            ]}}}
         CurveIndex=&Multiply_Float(A=$Particles.NormalizedAge, B=&RandomRangeFloat(Minimum=0.1, Maximum=0.5))
         Scale Curve=7
        )
        Alpha=0.7
       )
      ))
     2=&MakeVector2D(
      X=0.5
      Y=&FloatFromCurve(
       FloatCurve=DataInterface() {propertyValues {Curve {keys [<at -0.0046 0>,<at 0.0873 0.9701>,<at 1.0011 0.002>]}}}
       CurveIndex=&RandomRangeFloat(Minimum=0.5, Maximum=1)
      )
     )
     3=?
     4=?
     Index=&SpawnGroup()
     Num=::NewEnumerator1
    )
    Scale Sprite Size Mode=::NewEnumerator2
   )
   &GravityForce(Gravity=&MergeHelper_VectorSwitch(
     0=&RandomRangeVector(Minimum=<0 0 200>, Maximum=<0 0 -980>)
     1=&RandomRangeVector(Minimum=<0 0 350>, Maximum=<0 0 -1500>)
     2=&RandomRangeVector(Minimum=<0 0 200>, Maximum=<1 1 -980>)
     3=<0 0 0>
     4=<0 0 0>
     Index=&SpawnGroup()
     Num=::NewEnumerator3
    ))
   &CurlNoiseForce(
    Bias Noise=false
    Noise Frequency=&MergeHelper_FloatSwitch(0=50, 1=50, 2=0, 3=0, 4=0, Index=&SpawnGroup(), Num=::NewEnumerator3)
    Noise Strength=&MergeHelper_FloatSwitch(0=50, 1=50, 2=0, 3=0, 4=0, Index=&SpawnGroup(), Num=::NewEnumerator3)
    Pan Noise=false
   )
   &Drag(
    Drag=&MergeHelper_FloatSwitch(
     0=&RandomRangeFloat(Minimum=5, Maximum=30)
     1=&RandomRangeFloat(Minimum=5, Maximum=20)
     2=0
     3=0
     4=0
     Index=&SpawnGroup()
     Num=::NewEnumerator3
    )
    Rotational Drag:NiagaraFloat=&MergeHelper_FloatSwitch(0=0, 1=0, 2=0, 3=0, 4=0, Index=&SpawnGroup(), Num=::NewEnumerator3)
    Use Linear Drag=true
    Use Rotational Drag=true
   )
   &SolveForcesAndVelocity(Clamp Velocity=false, Limit Acceleration=false)
   &ScaleSpriteSizeBySpeed(
    Max Scale Factor=&MergeHelper_Vector2Switch(0=<1.2 3>, 1=<2 2>, 2=<1.2 1>, 3=<0 0>, 4=<0 0>, Index=&SpawnGroup(), Num=::NewEnumerator3)
    Min Scale Factor=&MergeHelper_Vector2Switch(0=<0.5 1.5>, 1=<1 1>, 2=<1 1.5>, 3=<0 0>, 4=<0 0>, Index=&SpawnGroup(), Num=::NewEnumerator3)
    Velocity Threshold:NiagaraFloat=&MergeHelper_FloatSwitch(0=250, 1=250, 2=250, 3=0, 4=0, Index=&SpawnGroup(), Num=::NewEnumerator3)
   )
   &ScaleColor(
    Scale Alpha=&MergeHelper_FloatSwitch(
     0=&FloatFromCurve(FloatCurve=DataInterface() {propertyValues {Curve {keys [<at -0.0092 0.0144>,<at 0.0967 0.9767>,<at 1.0011 0.002>]}}})
     1=&FloatFromCurve(FloatCurve=DataInterface() {propertyValues {Curve {keys [<at -0.0092 0.0144>,<at 0.0967 0.9767>,<at 1.0011 0.002>]}}})
     2=&FloatFromCurve(FloatCurve=DataInterface() {propertyValues {Curve {keys [<at -0.0092 0.0144>,<at 0.0967 0.9767>,<at 1.0011 0.002>]}}})
     3=1
     4=1
     Index=&SpawnGroup()
     Num=::NewEnumerator3
    )
    Scale RGB=&MergeHelper_VectorSwitch(0=<15 5 1>, 1=<15 5 1>, 2=<15 5 1>, 3=<1 1 1>, 4=<1 1 1>, Index=&SpawnGroup(), Num=::NewEnumerator3)
    ScaleA=true
    ScaleRGB=true
   )
   /SetVariables(
    Particles.Sparks Color:LinearColor=&MakeLinearColorFromVectorAndFloat("Vector (RGB)"=<15 5 1>, "Float (Alpha)"=1)
    Particles.Sparks SpriteSize:Vector2f=&Multiply_Vector2D_ByFloat(Float=1, Vector2D=&RandomRangeVector2D(Minimum=<0.1 0.5>, Maximum=<0.4 2.5>))
   ) {isSetParameters=true, module=SetVariables_#9}
  ]
  rendererClasses [&NiagaraSpriteRendererProperties]
  renderers [
   {rendererClass=&NiagaraSpriteRendererProperties}
   {rendererClass=&NiagaraSpriteRendererProperties, rendererIndex=1}
   {rendererClass=&NiagaraSpriteRendererProperties, rendererIndex=2}
  ]
 }
 Thruster_merged {
  bEnabled=false
  emitterUpdateScript [
   &EmitterState(Inactive Response=::NewEnumerator1, Life Cycle Mode=::NewEnumerator1, Loop Behavior=::NewEnumerator1, Loop Duration=4, UseLoopDelay=false)
   &SpawnBurst_Instantaneous(Spawn Count=2, Use Spawn Probability=false, UseLoopCountLimit=false)
   /Niagara/Noise/Waveform(
    EnableExponent1=false
    Global Amplitude Scale=0
    GlobalClampMax=false
    GlobalClampMin=false
    UseInput1=false
    UsePhase1=false
    "Waveform [1]":ENiagara_Waveforms=::NewEnumerator12
    "[1] Frequency"=7
   )
   /SetVariables(Emitter.Flicker:NiagaraFloat=&Add_Float(A=&Multiply_Float(A=$Emitter.Waveform.Waveform, B=0.5), B=0.5)) {isSetParameters=true}
   /SetVariables(Emitter.LightVisibility:NiagaraBool=/Niagara/DynamicInputs/Helpers/MaskBool(
     Mask Source:NiagaraInt32=1
     Mask Parameter:NiagaraInt32=&MakeCustomIntFromBool(Bool=$User.bIsFrontEndPreview, True INT=1, False INT=0)
     Mask:NiagaraInt32=1
     Masked Bool:NiagaraBool=false
     Passthrough Bool:NiagaraBool=&SetBoolByFloatComparison(A=$User.LightVisibility, B=0.1)
    )) {isSetParameters=true}
   /SetVariables(Emitter.ColorCoreScalar=10) {isSetParameters=true}
  ]
  particleSpawnScript [
   /SetVariables(Particles.VisibilityTag=&SpawnGroup()) {isSetParameters=true}
   &FN_SetSpawnGroup(Number of Spawn Groups=5, Set Ribbon ID and Ribbon Link Order=true) {enabled=false}
   &PartitionParticles(Number Of Partitions=5, Particles Per Partition=1, Partition Mode:ENiagara_PartitionMode_v2=::NewEnumerator1)
   &InitializeParticle(
    Color Mode=::NewEnumerator0
    Lifetime=&MergeHelper_FloatSwitch(
     0=1
     1=1
     2=?
     3=&Multiply_Float(
      A=&RandomRangeFloat(Minimum=0.4, Maximum=0.7)
      B=&FloatFromCurve(
       FloatCurve=DataInterface() {propertyValues {Curve {keys [<at 0 0.5>,<at 1.7 0.5>,<at 1.8 1>,<at 2 1>]}}}
       CurveIndex=$System.Age
       Scale Curve=2
      )
     )
     4=&Multiply_Float(
      A=&RandomRangeFloat(Minimum=0.4, Maximum=0.7)
      B=&FloatFromCurve(
       FloatCurve=DataInterface() {propertyValues {Curve {keys [<at 0 0.5>,<at 1.7 0.5>,<at 1.8 1>,<at 2 1>]}}}
       CurveIndex=$System.Age
       Scale Curve=2
      )
     )
     5=&Multiply_Float(
      A=&RandomRangeFloat(Minimum=0.4, Maximum=0.7)
      B=&FloatFromCurve(
       FloatCurve=DataInterface() {propertyValues {Curve {keys [<at 0 0.5>,<at 1.7 0.5>,<at 1.8 1>,<at 2 1>]}}}
       CurveIndex=$System.Age
       Scale Curve=2
      )
     )
     Index=&SpawnGroup()
     Num=::NewEnumerator0
    )
    Mesh Scale=&Multiply_VectorByFloat(Vector=<3.5 0.5 1.5>, Float=0.07)
    Mesh Scale Mode=::NewEnumerator4
    Sprite Rotation Mode=::NewEnumerator0
    Sprite Size=<90 175>
    Sprite Size Mode=::NewEnumerator4
    UsePositionOffset=false
    Write Material Random=true
   ) {module=InitializeParticle001}
   &InitialMeshOrientation(Mesh Axis To Orient=<0 0 1>, Mesh Orientation Mode=::NewEnumerator2, Orientation Vector=<0 1 0.6>, Use Rotation Vector=false)
   &CameraOffset(Camera Offset Amount=&MergeHelper_FloatSwitch(0=-5, 1=5, 2=0, 3=?, Index=0, Num=::NewEnumerator1))
   /SetVariables(Particles.GlobalRandom:NiagaraFloat=&RandomRangeFloat(Minimum=0.1, Maximum=0.23)) {isSetParameters=true}
  ]
  particleUpdateScript [
   &ParticleState(
    Kill Particles When Lifetime Has Elapsed=false
    Let Infinitely Lived Particles Die When Emitter Deactivates=true
    Loop Particles Lifetime=true
   )
   &SkeletalMeshLocation(
    Bone / Socket Index=$Particles.PartitionParticles.Partition
    MaskSpawnGroup=false
    Mesh Sampling Type=::NewEnumerator5
    Sampled Position Offset=&SelectVectorFromArray(
     Direct Array Index=&SpawnGroup()
     Vector Selection Array=DataInterface() {propertyValues {FloatData [<0 -5 -2>, <0 -5 -2>]}}
    )
    Sampled Scale Coordinate Space:ENiagaraCoordinateSpace=::Simulation
    Skeletal Mesh=$System.Skeletal Mesh
    Socket Sampling Mode=::NewEnumerator5
   )
   /SetVariables(
    Particles.Plasma scale:Vector3f=&Multiply_VectorByFloat(
     Vector=&Lerp_Vector(A=<0.12 0.12 0.15>, B=<0.12 0.12 0.15>, Alpha=1)
     Float=&FloatFromCurve(
      FloatCurve=DataInterface() {propertyValues {Curve {
         defaultValue=1
         keys [
          {interpMode=RCIM_Cubic, tangentMode=RCTM_User, time=0, value=0}
          {interpMode=RCIM_Cubic, tangentMode=RCTM_User, time=0.1089, value=1}
          <at 0.2301 1 RCIM_Cubic>
         ]
        }}}
      CurveIndex=$Emitter.Age
     )
    )
    Particles.plasma color:LinearColor=&MakeLinearColorFromVectorAndFloat(
     "Vector (RGB)"=&MakeVectorFromLinearColorRGB(Color=<rgba 45.6332 50 43.0383 1>)
     "Float (Alpha)"=&FloatFromCurve(FloatCurve=DataInterface() {propertyValues {Curve {keys [<at 0 0.5>,<at 0.9944 0>]}}})
    )
   ) {isSetParameters=true}
   /SetVariables(
    Particles.Glow color:LinearColor=&MakeLinearColorFromVectorAndFloat(
     "Vector (RGB)"=&Lerp_Vector(
      A=&MakeVectorFromLinearColorRGB(Color=<rgba 0.3042 4 0.3383 1>)
      B=&MakeVectorFromLinearColorRGB(Color=<rgba 0.3802 5 0.4228 1>)
      Alpha=1
     )
     "Float (Alpha)"=&FloatFromCurve(FloatCurve=DataInterface() {propertyValues {Curve {keys [<at 0 0.5>,<at 0.9944 0>]}}})
    )
    Particles.Glow size:Vector2f=&Lerp_Vector2(A=<45 45>, B=<60 60>, Alpha=1)
    Particles.fire 2 scale:Vector3f=<0.12 0.12 0.2>
   ) {isSetParameters=true}
   &SolveForcesAndVelocity(Clamp Velocity=false, Limit Acceleration=false)
   &ScaleColor(Scale Alpha=&Multiply_Float(A=$Emitter.Flicker, B=3), ScaleA=true, ScaleRGB=true)
   &DynamicMaterialParameters(
    Index 0 Param 1=&RandomRangeFloat(Minimum=0, Maximum=1)
    Index 0 Param 2=5
    Index 0 Param 4=&RandomRangeFloat(Minimum=-2, Maximum=-30)
    Module.Index0Param0Write=true
    Module.Index0Param1Write=false
    Module.Index0Param2Write=false
    Module.Index0Param3Write=true
   )
  ]
  rendererClasses [&NiagaraMeshRendererProperties]
  renderers [{rendererClass=&NiagaraMeshRendererProperties, rendererIndex=0}]
 }
 fire_v2001 {
  bEnabled=false
  emitterSpawnScript [/SetVariables(Emitter.FE Particle Spawn Rate=60, Emitter.Max Traversal Speed=50, Emitter.Particle Spawn Rate=60) {isSetParameters=true, module=SetVariables_#3}]
  emitterUpdateScript [
   &EmitterState()
   &SpawnBurst_Instantaneous(Spawn Count=2, Spawn Group=3, Use Spawn Probability=false, UseLoopCountLimit=false) {module=SpawnBurst_Instantaneous002}
   &SpawnBurst_Instantaneous(Spawn Group=5, Use Spawn Probability=false, UseLoopCountLimit=false) {module=SpawnBurst_Instantaneous001}
   &SpawnRate(SpawnGroup=4, SpawnRate=25, Use Spawn Probability=false) {module=SpawnRate004}
   &SpawnBurst_Instantaneous(Spawn Count=6, Spawn Group=2, Spawn Time=0.15, Use Spawn Probability=false, UseLoopCountLimit=false)
   &SpawnRate(SpawnGroup=1, SpawnRate=20, Use Spawn Probability=false) {module=SpawnRate001}
   &SpawnRate(
    SpawnRate=&LerpMultipleFloats(
     Float Count:ENiagara_MultipleLerpCount=::NewEnumerator3
     1:NiagaraFloat=&UniformRangedFloat(Minimum=15, Maximum=20)
     2:NiagaraFloat=4
     3:NiagaraFloat=?
     4:NiagaraFloat=?
     5:NiagaraFloat=?
     Alpha:NiagaraFloat=&ClampFloat(Float=&FN_LOD_Distance_Remap(Start Distance=600, End Distance=10000, Minimum Output=0, Maximum Output=1))
    )
    Use Spawn Probability=false
   )
   /SetVariables(Emitter.Flicker:NiagaraFloat=&Add_Float(A=&Multiply_Float(A=$Emitter.Waveform.Waveform, B=0.5), B=0.5)) {isSetParameters=true, module=SetVariables_#4}
  ]
  particleSpawnScript [
   /SetVariables(Particles.VisibilityTag=&SpawnGroup()) {isSetParameters=true, module=SetVariables_#5}
   &PartitionParticles()
   &InitializeParticle(
    Color=&MergeHelper_ColorSwitch(
     0=<rgba 0.0073 1 0 1>
     1=<rgba 0.0073 1 0 1>
     2=<rgba 0.076 1 0.0846 1>
     3=<rgba 1 1 1 0.1>
     4=$System.Global color
     5=<rgba 0.3715 1 0 1>
     Index=&SpawnGroup()
     Num=::NewEnumerator4
    )
    Lifetime=&MergeHelper_FloatSwitch(
     0=&Multiply_Float(A=&UniformRangedFloat(Minimum=0.9, Maximum=1.35), B=0.75)
     1=&RandomRangeFloat(Minimum=0.5, Maximum=0.6)
     2=0.7
     3=4
     4=1
     5=4
     Index=&SpawnGroup()
     Num=::NewEnumerator4
    )
    Mesh Scale=&FN_FrontEndVector3Modifier(Gameplay=<0.3 1.2 0.5>, Front End=<0.7 1 0.8>)
    Mesh Scale Mode=::NewEnumerator4
    Position Mode=::NewEnumerator1
    Sprite Rotation Angle=&MergeHelper_FloatSwitch(
     0=&UniformRangedFloat(Minimum=0, Maximum=180)
     1=&RandomRangeFloat(Minimum=-180, Maximum=200)
     2=&RandomRangeFloat(Minimum=-360, Maximum=360)
     3=100
     4=0
     Index=&SpawnGroup()
     Num=::NewEnumerator3
    )
    Sprite Rotation Mode=::NewEnumerator0
    Sprite Size=&MergeHelper_Vector2Switch(
     0=&Vector2DFromFloat(Value=&Multiply_Float(A=0.4, B=&UniformRangedFloat(Minimum=35, Maximum=50)))
     1=&Vector2DFromFloat(Value=&RandomRangeFloat(Minimum=5, Maximum=10))
     2=&RandomRangeVector2D(Minimum=<85 85>, Maximum=<175 175>)
     3=<5 5>
     4=&RandomRangeVector2D(Minimum=<20 20>, Maximum=<35 35>)
     Index=&SpawnGroup()
     Num=::NewEnumerator3
    )
    Sprite Size Mode=::NewEnumerator4
    UsePositionOffset=false
    Write Material Random=true
   ) {module=InitializeParticle001}
   &ShapeLocation(
    MaskSpawnGroup=false
    Override Seed=false
    Sphere Radius=&FloatFromCurve(FloatCurve=DataInterface() {propertyValues {Curve {keys [<at 0 0>,<at 1 5>]}}})
    UseOwnerScale=true
   )
   &CameraOffset(Camera Offset Amount=5)
   &AddVelocity(Velocity=&MergeHelper_VectorSwitch(
     0=&RandomRangeVector(Minimum=<0 -50 20>, Maximum=<1 -85 20>)
     1=&RandomRangeVector(Minimum=<0 -75 20>, Maximum=<1 -100 20>)
     2=&MakeVector(
      X=0
      Y=-100
      Z=&FloatFromCurve(FloatCurve=DataInterface() {propertyValues {Curve {keys [<at 0 1>,<at 0.75 0>]}}}, Scale Curve=20)
     )
     3=<0 0 0>
     4=&RandomRangeVector(Minimum=<0 -50 20>, Maximum=<1 -100 20>)
     5=<0 0 0>
     Index=&SpawnGroup()
     Num=::NewEnumerator4
    )) {module=AddVelocity002}
   /SetVariables(
    Particles.1 dynparam:Vector4f=&MakeVector4(X=0, Y=&RandomRangeFloat(Minimum=0, Maximum=4), Z=0, W=2.5)
    Particles.New fire DynParam 001:Vector4f=&MakeVector4(X=&RandomRangeFloat(Minimum=0, Maximum=4), Y=0, Z=0, W=2.5)
   ) {isSetParameters=true, module=SetVariables_#6}
   &InitialMeshOrientation(Mesh Axis To Orient=<0 0 1>, Mesh Orientation Mode=::NewEnumerator2, Orientation Vector=<0 -0.8 1>, Use Rotation Vector=false)
  ]
  particleUpdateScript [
   &ParticleState()
   &SubUVAnimation(
    Random Start Frame=true
    "Sprite Renderer "=DataInterface() {propertyValues {SpriteRenderer=&self:fire_v2001_0.NiagaraSpriteRendererProperties_5}}
    UseEndFrame=false
    UseStartFrame=false
   )
   &ScaleSpriteSize(
    Scale Factor=&MergeHelper_Vector2Switch(
     0=&Vector2DFromCurve(
      Vector2Curve=DataInterface() {propertyValues {
        XCurve {keys [<at 0 0.3 RCIM_Cubic>, {arriveTangent=0.3, interpMode=RCIM_Cubic, leaveTangent=0.3, time=0.3, value=1}, <at 1 0.6>]}
        YCurve {keys [<at 0 0.3 RCIM_Cubic>, {arriveTangent=0.3, interpMode=RCIM_Cubic, leaveTangent=0.3, time=0.3, value=1}, <at 1 0.6>]}
       }}
      ScaleCurve=<1 1>
     )
     1=&Multiply_Vector2D_ByFloat(
      Float=&FloatFromCurve(FloatCurve=DataInterface() {propertyValues {Curve {keys [
           {arriveTangent=1.4708, interpMode=RCIM_Cubic, leaveTangent=1.4708, tangentMode=RCTM_User, time=0, value=0.3}
           {arriveTangent=1.7, interpMode=RCIM_Cubic, leaveTangent=1.7, time=0.6, value=1.625}
           {interpMode=RCIM_Cubic, tangentMode=RCTM_User, time=1, value=2}
          ]}}})
      Vector2D=<1 1>
     )
     2=&Multiply_Vector2D_ByFloat(
      Float=&FloatFromCurve(
       FloatCurve=DataInterface() {propertyValues {Curve {keys [
           {arriveTangent=1.4708, interpMode=RCIM_Cubic, leaveTangent=1.4708, tangentMode=RCTM_User, time=0, value=0}
           {arriveTangent=-0.0652, interpMode=RCIM_Cubic, leaveTangent=-0.0652, tangentMode=RCTM_User, time=0.2977, value=1.7547}
           {interpMode=RCIM_Cubic, tangentMode=RCTM_User, time=1, value=2}
          ]}}}
       Scale Curve=0.15
      )
      Vector2D=<1 1>
     )
     3=<1 1>
     4=&Vector2DFromFloat(Value=&FloatFromCurve(FloatCurve=DataInterface() {propertyValues {Curve {keys [
           {arriveTangent=0.8753, interpMode=RCIM_Cubic, leaveTangent=0.8753, tangentMode=RCTM_User, time=0, value=0}
           {arriveTangent=0.0238, interpMode=RCIM_Cubic, leaveTangent=0.0238, tangentMode=RCTM_User, time=1, value=1.7}
          ]}}}))
     Index=&SpawnGroup()
     Num=::NewEnumerator3
    )
    Scale Sprite Size Mode=::NewEnumerator2
   )
   &SkeletalMeshLocation(
    Bone / Socket Index=$Particles.PartitionParticles.Partition
    MaskSpawnGroup=false
    Mesh Sampling Type=::NewEnumerator5
    Position Sampling=::NewEnumerator1
    Sampled Position Offset=&MergeHelper_VectorSwitch(
     0=<0 3 -3>
     1=<0 7 -3>
     2=<0 7 -3>
     3=<0 7 0>
     4=<0 7 -3>
     5=<0 7 -3>
     Index=&SpawnGroup()
     Num=::NewEnumerator4
    )
    Sampled Position Offset Coordinate Space:ENiagaraSkelMeshTransforms=::NewEnumerator2
    Skeletal Mesh=$System.Skeletal Mesh
    Socket Sampling Mode=::NewEnumerator5
   )
   &ScaleMeshSize(Scale Factor=&MakeVector(
     X=0.5
     Y=&FloatFromCurve(
      FloatCurve=DataInterface() {propertyValues {Curve {
         defaultValue=1
         keys [{interpMode=RCIM_Cubic, tangentMode=RCTM_User, time=0, value=0}, {interpMode=RCIM_Cubic, tangentMode=RCTM_User, time=0.05, value=1}]
        }}}
      Scale Curve=0.2
     )
     Z=0.5
    ))
   &Drag(
    Drag=&MergeHelper_FloatSwitch(
     0=3
     1=&FloatFromCurve(FloatCurve=DataInterface() {propertyValues {Curve {keys [<at 0 0.075 RCIM_Cubic>,<at 1 1.075 RCIM_Cubic>]}}}, Scale Curve=10)
     2=0
     3=0
     4=0
     Index=&SpawnGroup()
     Num=::NewEnumerator3
    )
    Rotational Drag=5
    Use Linear Drag=true
    Use Rotational Drag=true
   )
   &AccelerationForce(
    Acceleration=&MergeHelper_VectorSwitch(
     0=&UniformRangedVector(Minimum=<-20 -147 30>, Maximum=<20 -205 30>)
     1=<0 0 0>
     2=<0 0 0>
     3=<0 0 0>
     4=<0 0 0>
     Index=&SpawnGroup()
     Num=::NewEnumerator3
    )
    Coordinate Space=::Local
   )
   &SolveForcesAndVelocity(Clamp Velocity=false, Limit Acceleration=false)
   &ScaleSpriteSizeBySpeed(
    Sample Scale Factor By Curve=true
    Scale Factor Curve=DataInterface() {propertyValues {Curve {keys [<at 0 0>,<at 1 2>]}}}
    Velocity Threshold=100
   )
   &ScaleColor(
    Scale Alpha=&MergeHelper_FloatSwitch(
     0=&Multiply_Float(
      A=0.5
      B=&FloatFromCurve(FloatCurve=DataInterface() {propertyValues {Curve {keys [
           {arriveTangent=2.4553, interpMode=RCIM_Cubic, leaveTangent=2.4553, time=0, value=0}
           {arriveTangent=-0.0116, interpMode=RCIM_Cubic, leaveTangent=-0.0116, tangentMode=RCTM_User, time=0.4073, value=1}
           {arriveTangent=-2.0743, interpMode=RCIM_Cubic, leaveTangent=-2.0743, tangentMode=RCTM_User, time=0.7, value=0.2}
           {arriveTangent=-0.6667, interpMode=RCIM_Cubic, leaveTangent=-0.6667, time=1, value=0}
          ]}}})
     )
     1=&FloatFromCurve(
      FloatCurve=DataInterface() {propertyValues {Curve {keys [
          {arriveTangent=6.7724, interpMode=RCIM_Cubic, leaveTangent=6.7724, tangentMode=RCTM_User, time=0, value=0}
          {interpMode=RCIM_Cubic, tangentMode=RCTM_User, time=0.3949, value=1}
          {arriveTangent=-1.6527, interpMode=RCIM_Cubic, leaveTangent=-1.6527, time=0.7199, value=0.2109}
          {arriveTangent=-0.174, interpMode=RCIM_Cubic, leaveTangent=-0.174, tangentMode=RCTM_User, time=1, value=0}
         ]}}}
      Scale Curve=&FNTimeOfDayScalar(Night Scale=1.5, Morning Scale=0.5, Day Scale=0.52, Evening Scale=0.5)
     )
     2=&FloatFromCurve(
      FloatCurve=DataInterface() {propertyValues {Curve {keys [
          {arriveTangent=6.7724, interpMode=RCIM_Cubic, leaveTangent=6.7724, tangentMode=RCTM_User, time=0, value=0}
          {interpMode=RCIM_Cubic, tangentMode=RCTM_User, time=0.25, value=1}
          {arriveTangent=-0.174, interpMode=RCIM_Cubic, leaveTangent=-0.174, tangentMode=RCTM_User, time=0.6, value=0}
         ]}}}
      Scale Curve=&FNTimeOfDayScalar(Night Scale=1.5, Morning Scale=0.5, Day Scale=0.52, Evening Scale=0.5)
     )
     3=1
     4=&FloatFromCurve(
      FloatCurve=DataInterface() {propertyValues {Curve {keys [
          {arriveTangent=0.0878, interpMode=RCIM_Cubic, leaveTangent=0.0878, tangentMode=RCTM_User, time=0, value=0}
          {
           arriveTangent=0.0878
           arriveTangentWeight=0.0754
           interpMode=RCIM_Cubic
           leaveTangent=0.0878
           leaveTangentWeight=0.2297
           tangentMode=RCTM_User
           time=0.1684
           value=1
          }
          {arriveTangent=-0.3893, interpMode=RCIM_Cubic, leaveTangent=-3.5553, tangentMode=RCTM_Break, time=0.4045, value=1}
          {arriveTangent=-0.0368, interpMode=RCIM_Cubic, leaveTangent=-2.9749, tangentMode=RCTM_Break, time=0.9, value=0}
         ]}}}
      Scale Curve=&FN_FrontEndFloatModifier(
       Gameplay=&FNTimeOfDayScalar(Night Scale=0.15, Morning Scale=0.1, Day Scale=0.1, Evening Scale=0.1)
       Front End=0.4
      )
     )
     Index=&SpawnGroup()
     Num=::NewEnumerator3
    )
    Scale RGB=&MergeHelper_VectorSwitch(
     0=&VectorFromCurve(VectorCurve=DataInterface() {propertyValues {
        XCurve {keys [<at 0.3242 3 RCIM_Cubic>,<at 0.9878 0.9731>]}
        YCurve {keys [<at 0.3242 2.8 RCIM_Cubic>,<at 1 0.9731>]}
        ZCurve {keys [<at 0.3169 2.6605 RCIM_Cubic>,<at 1 0.9731>]}
       }})
     1=<1 1 1>
     2=&MakeVectorFromLinearColorRGB(Color=&ColorFromCurve(ColorCurve=DataInterface() {propertyValues {
         AlphaCurve {keys [{arriveTangent=-0.1596, interpMode=RCIM_Cubic, leaveTangent=-0.1596, tangentMode=RCTM_User, time=0.8987, value=1}, <at 0.982 0 RCIM_Cubic>]}
         BlueCurve {keys [<at 0.0805 0.6018>,<at 0.9823 0.873>]}
         GreenCurve {keys [<at 0.0805 1>,<at 0.9823 1>]}
         RedCurve {keys [<at 0.0805 0.4511>,<at 0.9823 0.0611>]}
        }}))
     3=<0 1 0>
     4=&VectorFromFloat(Value=&RandomRangeFloat(Minimum=0.3, Maximum=0.5))
     Index=&SpawnGroup()
     Num=::NewEnumerator3
    )
    ScaleA=true
    ScaleRGB=true
   ) {module=ScaleColor001}
   /SetVariables(
    Particles.4 - smoke DynParam:Vector4f=&MakeVector4(
     X=&Add_Float(
      A=&RandomRangeFloat(Minimum=0, Maximum=1)
      B=&FloatFromCurve(
       FloatCurve=DataInterface() {propertyValues {Curve {keys [
           <at 0 0 RCIM_Cubic>
           {arriveTangent=1.5947, interpMode=RCIM_Cubic, leaveTangent=1.5947, tangentMode=RCTM_User, time=0.2557, value=0.5257}
           <at 1 1 RCIM_Cubic>
          ]}}}
       CurveIndex=$Particles.Age
       Scale Curve=0.25
      )
     )
     Y=0
     Z=0
     W=0
    )
    Particles.new fire DynParam:Vector4f=&MakeVector4(
     X=&FloatFromCurve(FloatCurve=DataInterface() {propertyValues {Curve {keys [<at 0 1>]}}})
     Y=0.5
     Z=&FloatFromCurve(FloatCurve=DataInterface() {propertyValues {Curve {keys [<at 0 0.5 RCIM_Cubic>, {arriveTangent=2.0035, interpMode=RCIM_Cubic, leaveTangent=2.0035, time=0.2516, value=0.4588}, <at 0.7282 1.9588>]}}})
     W=2
    )
   ) {isSetParameters=true, module=SetVariables_#7}
   &SpriteRotationRate(Rotation Rate=&Lerp_Float(A=-10, B=10, Alpha=$Particles.Initial.MaterialRandom)) {module=SpriteRotationRate001}
   &FN_Light_Attributes(
    Color=<rgba 0.609 5 1.1003 1>
    Create Color Attribute=true
    Create Position Attribute=true
    Create Radius Attribute=true
    Enable Particle Light Enable=true
    Enable Visibility Tag=true
    Light Enabled=true
    Position=&SelectVectorFromArray(
     Direct Array Index=$Particles.PartitionParticles.Partition
     Vector Selection Array=DataInterface() {propertyValues {FloatData [<0 -68 10>, <16 -55 10>]}}
    )
    Radius=50
    Scale Color=&Multiply_Float(
     A=0.5
     B=&Multiply_Float(
      A=/Niagara/DynamicInputs/Helpers/Waveform("Waveform [1]"=::NewEnumerator0, "[1] Frequency"=3, "[1] Amplitude Min/Max"=<0.7 1>, "[1] Scale"=2.6368)
      B=&FNTimeOfDayScalar(Night Scale=0.5, Morning Scale=0.6, Day Scale=0.45, Evening Scale=0.6)
     )
    )
    VisibilityTag=5
   )
   &SubUVAnimation(
    Random Start Frame=true
    "Sprite Renderer "=DataInterface() {propertyValues {SpriteRenderer=&self:fire_v2001_0.NiagaraSpriteRendererProperties_3}}
    UseEndFrame=false
    UseStartFrame=false
   ) {module=SubUVAnimation001}
  ]
  rendererClasses [
   &NiagaraSpriteRendererProperties
   &NiagaraMeshRendererProperties
   &NiagaraLightRendererProperties
  ]
  renderers [
   {rendererClass=&NiagaraSpriteRendererProperties}
   {rendererClass=&NiagaraSpriteRendererProperties, rendererIndex=1}
   {rendererClass=&NiagaraSpriteRendererProperties, rendererIndex=2}
   {rendererClass=&NiagaraMeshRendererProperties, rendererIndex=3}
   {rendererClass=&NiagaraSpriteRendererProperties, rendererIndex=4}
   {rendererClass=&NiagaraLightRendererProperties, rendererIndex=5}
  ]
 }
 metal_ground_flares {
  emitterUpdateScript [
   &EmitterState()
   &SpawnRate(
    SpawnGroup=3
    SpawnRate=&FN_FrontEndFloatModifier(
     Gameplay=&MakeCustomFloatFromBool(
      Bool=&SetBoolByFloatComparison(A=&VectorLength(Vector=$User.OwnerLocalVelocity), B=100)
      True Float=&FloatFromCurve(
       FloatCurve=DataInterface() {propertyValues {Curve {keys [
           {interpMode=RCIM_Cubic, tangentMode=RCTM_User, time=0, value=0}
           {interpMode=RCIM_Cubic, tangentMode=RCTM_User, time=0.0341, value=100}
           {interpMode=RCIM_Cubic, tangentMode=RCTM_User, time=0.8049, value=100.8627}
           {interpMode=RCIM_Cubic, tangentMode=RCTM_User, time=0.9514, value=-0.9412}
          ]}}}
       CurveIndex=$Emitter.NormalizedLoopAge
      )
      False Float=0
     )
     Front End=&FloatFromCurve(
      FloatCurve=DataInterface() {propertyValues {Curve {keys [
          {interpMode=RCIM_Cubic, tangentMode=RCTM_User, time=0, value=0}
          {interpMode=RCIM_Cubic, tangentMode=RCTM_User, time=0.0341, value=100}
          {interpMode=RCIM_Cubic, tangentMode=RCTM_User, time=0.8049, value=100.8627}
          {interpMode=RCIM_Cubic, tangentMode=RCTM_User, time=0.9514, value=-0.9412}
         ]}}}
      CurveIndex=$Emitter.NormalizedLoopAge
      Scale Curve=3
     )
    )
    Use Spawn Probability=false
   ) {module=SpawnRate004}
   &SpawnRate(
    SpawnGroup=2
    SpawnRate=&FN_FrontEndFloatModifier(
     Gameplay=&MakeCustomFloatFromBool(
      Bool=&SetBoolByFloatComparison(A=&VectorLength(Vector=$User.OwnerLocalVelocity), B=100)
      True Float=&FloatFromCurve(
       FloatCurve=DataInterface() {propertyValues {Curve {keys [
           {interpMode=RCIM_Cubic, tangentMode=RCTM_User, time=0, value=0}
           {interpMode=RCIM_Cubic, tangentMode=RCTM_User, time=0.0341, value=100}
           {interpMode=RCIM_Cubic, tangentMode=RCTM_User, time=0.8049, value=100.8627}
           {interpMode=RCIM_Cubic, tangentMode=RCTM_User, time=0.9514, value=-0.9412}
          ]}}}
       CurveIndex=$Emitter.NormalizedLoopAge
      )
      False Float=0
     )
     Front End=&FloatFromCurve(
      FloatCurve=DataInterface() {propertyValues {Curve {keys [
          {interpMode=RCIM_Cubic, tangentMode=RCTM_User, time=0, value=0}
          {interpMode=RCIM_Cubic, tangentMode=RCTM_User, time=0.0341, value=100}
          {interpMode=RCIM_Cubic, tangentMode=RCTM_User, time=0.8049, value=100.8627}
          {interpMode=RCIM_Cubic, tangentMode=RCTM_User, time=0.9514, value=-0.9412}
         ]}}}
      CurveIndex=$Emitter.NormalizedLoopAge
      Scale Curve=3
     )
    )
    Use Spawn Probability=false
   ) {module=SpawnRate001}
   &SpawnRate(
    SpawnGroup=1
    SpawnRate=&FN_FrontEndFloatModifier(
     Gameplay=&MakeCustomFloatFromBool(
      Bool=&SetBoolByFloatComparison(A=&VectorLength(Vector=$User.OwnerLocalVelocity), B=100)
      True Float=30
      False Float=0
     )
     Front End=3
    )
    Use Spawn Probability=false
   ) {module=SpawnRate003}
   &SpawnRate(
    SpawnRate=&FN_FrontEndFloatModifier(
     Gameplay=&MakeCustomFloatFromBool(
      Bool=&SetBoolByFloatComparison(A=&VectorLength(Vector=$User.OwnerLocalVelocity), B=100)
      True Float=350
      False Float=0
     )
     Front End=1
    )
    Use Spawn Probability=false
   ) {module=SpawnRate002}
  ]
  particleSpawnScript [
   /SetVariables(Particles.VisibilityTag=&SpawnGroup()) {isSetParameters=true}
   /SetVariables(Particles.Custom Random:NiagaraFloat=&RandomRangeFloat(Minimum=0.2, Maximum=0.8)) {isSetParameters=true}
   &InitializeParticle(
    Color Mode=::NewEnumerator0
    Lifetime=&MergeHelper_FloatSwitch(
     0=&RandomRangeFloat(Minimum=0.05, Maximum=0.05)
     1=&RandomRangeFloat(Minimum=0.5, Maximum=0.052)
     2=&RandomRangeFloat(Minimum=0.01, Maximum=0.04)
     3=&RandomRangeFloat(Minimum=0.01, Maximum=0.04)
     Index=&SpawnGroup()
     Num=::NewEnumerator2
    )
    Sprite Rotation Mode=::NewEnumerator2
    Sprite Size Mode=::NewEnumerator3
    Uniform Sprite Size=&MergeHelper_FloatSwitch(
     0=30
     1=30
     2=&RandomRangeFloat(Minimum=5, Maximum=15, Evaluation Type=::NewEnumerator1)
     3=&RandomRangeFloat(Minimum=5, Maximum=15, Evaluation Type=::NewEnumerator1)
     Index=&SpawnGroup()
     Num=::NewEnumerator2
    )
    UsePositionOffset=false
    Write Material Random=true
   )
   &CameraOffset(Camera Offset Amount=6)
   /SetVariables(Particles.lower sprite SpriteFacing:Vector3f=<0.05 0.05 1>) {isSetParameters=true}
  ]
  particleUpdateScript [
   &ParticleState()
   &SkeletalMeshLocation(
    Bone Sampling Mode=::NewEnumerator4
    MaskSpawnGroup=false
    Sampled Position Offset=&FN_FrontEndVector3Modifier(Gameplay=<0 -60 3>, Front End=<0 -60 2>)
    Sampled Position Offset Coordinate Space:ENiagaraSkelMeshTransforms=::NewEnumerator2
    Skeletal Mesh=$System.Skeletal Mesh
   )
   &ScaleSpriteSize(
    Scale Sprite Size Mode=::NewEnumerator0
    Uniform Scale Factor=&MergeHelper_FloatSwitch(
     0=&FloatFromCurve(FloatCurve=DataInterface() {propertyValues {Curve {keys [<at 0 0>,<at 1 1>]}}})
     1=&FloatFromCurve(FloatCurve=DataInterface() {propertyValues {Curve {keys [<at 0 0>,<at 1 1>]}}}, Scale Curve=3)
     2=&FloatFromCurve(FloatCurve=DataInterface() {propertyValues {Curve {keys [<at 0 0>,<at 1 0.75>]}}})
     3=&FloatFromCurve(
      FloatCurve=DataInterface() {propertyValues {Curve {keys [
          {interpMode=RCIM_Cubic, tangentMode=RCTM_Break, time=0, value=0}
          {interpMode=RCIM_Cubic, tangentMode=RCTM_Break, time=0.1692, value=0.7365}
          {interpMode=RCIM_Cubic, tangentMode=RCTM_Break, time=0.2971, value=0.2347}
          {interpMode=RCIM_Cubic, tangentMode=RCTM_Break, time=0.412, value=0.8586}
          {interpMode=RCIM_Cubic, tangentMode=RCTM_Break, time=0.5752, value=0.2192}
          {interpMode=RCIM_Cubic, tangentMode=RCTM_Break, time=0.6474, value=0.7557}
          {interpMode=RCIM_Cubic, tangentMode=RCTM_Break, time=0.7809, value=0.0771}
          {interpMode=RCIM_Cubic, tangentMode=RCTM_Break, time=0.8773, value=0.7406}
          {interpMode=RCIM_Cubic, tangentMode=RCTM_Break, time=1, value=0}
         ]}}}
      Scale Curve=3
     )
     Index=&SpawnGroup()
     Num=::NewEnumerator2
    )
   )
   &SpriteRotationRate(Rotation Rate=&RandomRangeFloat(Minimum=-25, Maximum=25))
   /SetVariables(Particles.Flicker:NiagaraFloat=&Waveform(
     "Waveform [1]"=::NewEnumerator0
     "[1] Frequency"=&Lerp_Float(A=4, B=6, Alpha=$Particles.Custom Random)
     "[1] Amplitude Min/Max"=&MergeHelper_Vector2Switch(0=<0 1>, 1=<0 0.3>, 2=?, 3=?, 4=?, Index=&SpawnGroup(), Num=::NewEnumerator0)
     "[1] Bias"=&RandomRangeFloat(Minimum=-0.08, Maximum=0, Evaluation Type=::NewEnumerator1)
    )) {isSetParameters=true}
   &SubUVAnimation(
    End Frame Range Override=2
    "Sprite Renderer "=DataInterface() {propertyValues {SpriteRenderer=&self:metal_ground_flares_0.NiagaraSpriteRendererProperties_3}}
    UseEndFrame=false
    UseStartFrame=false
   )
   &Drag(Drag=50, Use Linear Drag=true, Use Rotational Drag=true)
   &SolveForcesAndVelocity(Clamp Velocity=false, Limit Acceleration=false)
   /SetVariables(
    Particles.Flare Color:LinearColor=&MakeLinearColorFromVectorAndFloat(
     "Vector (RGB)"=&Lerp_Vector(
      A=&MakeVectorFromLinearColorRGB(Color=<rgba 1 0.2274 0.0008 1>)
      B=&MakeVectorFromLinearColorRGB(Color=<rgba 1 0.0985 0 1>)
      Alpha=$Particles.Flicker
     )
     "Float (Alpha)"=&Multiply_Float(
      A=&FloatFromCurve(FloatCurve=DataInterface() {propertyValues {Curve {
          defaultValue=1
          keys [
           {arriveTangent=17.4943, interpMode=RCIM_Cubic, leaveTangent=17.4943, tangentMode=RCTM_User, time=0, value=0}
           {arriveTangentWeight=0.05, interpMode=RCIM_Cubic, leaveTangentWeight=0.2333, tangentMode=RCTM_User, time=0.15, value=1}
           {
            arriveTangent=-0.9301
            arriveTangentWeight=0.3603
            interpMode=RCIM_Cubic
            leaveTangent=-0.9301
            leaveTangentWeight=0.0772
            tangentMode=RCTM_User
            time=0.8279
            value=0.6565
           }
           {arriveTangent=-5.0054, tangentMode=RCTM_User, time=0.9828, value=0}
          ]
         }}})
      B=$Particles.Flicker
     )
    )
    Particles.floor flare color:LinearColor=&MakeLinearColorFromVectorAndFloat(
     "Vector (RGB)"=&Lerp_Vector(
      A=&MakeVectorFromLinearColorRGB(Color=<rgba 4.3698 2.6465 0.4346 1>)
      B=&MakeVectorFromLinearColorRGB(Color=<rgba 4.3698 1.9199 0.4158 1>)
      Alpha=$Particles.Flicker
     )
     "Float (Alpha)"=&Multiply_Float(
      A=&FloatFromCurve(
       FloatCurve=DataInterface() {propertyValues {Curve {
          defaultValue=1
          keys [
           {arriveTangent=17.4943, interpMode=RCIM_Cubic, leaveTangent=17.4943, tangentMode=RCTM_User, time=0, value=0}
           {arriveTangentWeight=0.05, interpMode=RCIM_Cubic, leaveTangentWeight=0.2333, tangentMode=RCTM_User, time=0.15, value=1}
           {
            arriveTangent=-0.9301
            arriveTangentWeight=0.3603
            interpMode=RCIM_Cubic
            leaveTangent=-0.9301
            leaveTangentWeight=0.0772
            tangentMode=RCTM_User
            time=0.8279
            value=0.6565
           }
           {arriveTangent=-5.0054, tangentMode=RCTM_User, time=0.9828, value=0}
          ]
         }}}
       Scale Curve=0.5
      )
      B=$Particles.Flicker
     )
    )
   ) {isSetParameters=true}
   /SetVariables(Particles.Glow Color=&MakeLinearColorFromVectorAndFloat(
     "Vector (RGB)"=&MakeVectorFromLinearColorRGB(Color=&LerpLinearColors(StartColor=<rgba 4.3698 2.6465 0.4346 1>, EndColor=<rgba 4.3698 1.495 0.5068 1>, LerpFactor=$Particles.Flicker))
     "Float (Alpha)"=&Multiply_Float(
      A=&FloatFromCurve(FloatCurve=DataInterface() {propertyValues {Curve {
          defaultValue=1
          keys [
           {arriveTangent=17.4943, interpMode=RCIM_Cubic, leaveTangent=17.4943, tangentMode=RCTM_User, time=0, value=0}
           {arriveTangentWeight=0.05, interpMode=RCIM_Cubic, leaveTangentWeight=0.2333, tangentMode=RCTM_User, time=0.0669, value=0.9624}
           {
            arriveTangent=-0.8996
            arriveTangentWeight=0.3603
            interpMode=RCIM_Cubic
            leaveTangent=-0.8996
            leaveTangentWeight=0.0772
            tangentMode=RCTM_User
            time=0.8524
            value=0.7765
           }
           {arriveTangent=-10.1516, tangentMode=RCTM_User, time=0.9863, value=0}
          ]
         }}})
      B=$Particles.Flicker
     )
    )) {isSetParameters=true}
   /SetVariables(Particles.hot core center:LinearColor=&MakeLinearColorFromVectorAndFloat(
     "Vector (RGB)"=&Multiply_Vector(
      A=&VectorFromFloat(Value=&MakeCustomFloatFromBool(
        Bool=$User.bIsFrontEnd
        True Float=1
        False Float=&Multiply_Float(
         A=&FNTimeOfDayScalar(Night Scale=0.75, Morning Scale=0.5, Evening Scale=0.5)
         B=&Lerp_Float(A=1, B=&RandomRangeFloat(Minimum=1.25, Maximum=2), Alpha=$User.Intensity)
        )
       ))
      B=&MakeVectorFromLinearColorRGB(Color=<rgba 4.3698 2.1331 0.6575 1>)
     )
     "Float (Alpha)"=1
    )) {isSetParameters=true}
   /SetVariables(Particles.bottom SpriteSize:Vector2f=&Vector2DFromFloat(Value=&FloatFromCurve(
      FloatCurve=DataInterface() {propertyValues {Curve {keys [
          {interpMode=RCIM_Cubic, tangentMode=RCTM_Break, time=0, value=0}
          {interpMode=RCIM_Cubic, tangentMode=RCTM_Break, time=0.1692, value=0.7365}
          {interpMode=RCIM_Cubic, tangentMode=RCTM_Break, time=0.2971, value=0.2347}
          {interpMode=RCIM_Cubic, tangentMode=RCTM_Break, time=0.412, value=0.8586}
          {interpMode=RCIM_Cubic, tangentMode=RCTM_Break, time=0.5752, value=0.2192}
          {interpMode=RCIM_Cubic, tangentMode=RCTM_Break, time=0.6474, value=0.7557}
          {interpMode=RCIM_Cubic, tangentMode=RCTM_Break, time=0.7809, value=0.0771}
          {interpMode=RCIM_Cubic, tangentMode=RCTM_Break, time=0.8773, value=0.7406}
          {interpMode=RCIM_Cubic, tangentMode=RCTM_Break, time=1, value=0}
         ]}}}
      Scale Curve=5
     ))) {isSetParameters=true}
  ]
  rendererClasses [&NiagaraSpriteRendererProperties]
  renderers [
   {rendererClass=&NiagaraSpriteRendererProperties}
   {rendererClass=&NiagaraSpriteRendererProperties, rendererIndex=1}
   {rendererClass=&NiagaraSpriteRendererProperties, rendererIndex=2}
   {rendererClass=&NiagaraSpriteRendererProperties, rendererIndex=3}
  ]
 }
}
summary {
 systemName=NS_Emote_ChimeCurlTell_wheelie_sparks
 emitters [
  {emitterName=Dust, rendererClasses [&NiagaraSpriteRendererProperties]}
  {emitterName=Sparks_condensed, simTarget=GPUComputeSim, rendererClasses [&NiagaraSpriteRendererProperties]}
  {bEnabled=false, emitterName=Thruster_merged, rendererClasses [&NiagaraMeshRendererProperties]}
  {
   bEnabled=false
   emitterName=fire_v2001
   rendererClasses [
    &NiagaraSpriteRendererProperties
    &NiagaraMeshRendererProperties
    &NiagaraLightRendererProperties
   ]
  }
  {emitterName=metal_ground_flares, rendererClasses [&NiagaraSpriteRendererProperties]}
  {emitterName=Sparks_condensed001, rendererClasses [&NiagaraSpriteRendererProperties]}
  {
   emitterName=Exhaust_fire_condensed
   rendererClasses [
    &NiagaraSpriteRendererProperties
    &NiagaraMeshRendererProperties
    &NiagaraLightRendererProperties
   ]
  }
 ]
 userVariables [
  User.Anim_curve_spawnRate {type=NiagaraFloat, defaultValue=35, description=""}
  User.bIsFrontEnd {type=NiagaraBool}
  User.bIsFrontEndPreview {type=NiagaraBool}
  User.Intensity {type=NiagaraFloat, defaultValue=5, description=""}
  User.LightVisibility {type=NiagaraFloat, defaultValue=1, description=""}
  User.OwnerLocalVelocity {type=Vector3f, description="", defaultValue {struct=Vector3f, value=<0 0 0>}}
  User.ScaleSize {type=Vector3f, description="", defaultValue {struct=Vector3f, value=<0 0 0>}}
 ]
}
